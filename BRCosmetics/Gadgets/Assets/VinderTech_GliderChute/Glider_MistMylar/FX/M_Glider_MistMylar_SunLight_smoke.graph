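# Translucent unlit Niagara sprite material for sunlit smoke effects on the MistMylar glider.
# A shared raymarched smoke function supplies emissive and opacity.
# The emissive is desaturated by a parameter (fully by default), tinted by particle color, then passed through the shared FX emissive adjustments function.
# The raymarched opacity is multiplied by a selectable channel, green by default, of an 8x8 smoke texture parameter.
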Material /BRCosmetics/Gadgets/Assets/VinderTech_GliderChute/Glider_MistMylar/FX/M_Glider_MistMylar_SunLight_smoke
graph {
 properties {
  bAutomaticallySetUsageInEditor=false
  bUsedWithNiagaraSprites=true
  blendMode=BLEND_Translucent
  pixelDepthOffsetMode=PDOM_Legacy
  shadingModel=MSM_Unlit
  usedShadingModels=Unlit
 }
 nodes [
  n1 = MaterialFunctionCall() {materialFunction=/Game/Effects/Raymarch/MF_Raymarched_Smoke_Func}
  n2 = ScalarParameter() {defaultValue=1, parameterName=Desaturation}
  MaterialFunctionCall(Emissive=Multiply(A=ParticleColor().RGB, B=Desaturation(None=@n1.Emissive, Fraction=@n2))) {materialFunction=&MF_FX_EmissiveAdjustments}
  n3 = TextureSampleParameter2D() {parameterName=Default_Smoke, samplerType=SAMPLERTYPE_LinearColor, texture=/Game/Effects/Fort_Effects/Textures/Smoke/T_Smoke_8x8_01}
  Multiply(A=@n1.Opacity, B=StaticComponentMaskParameter(None=@n3.RGBA) {defaultG=true, parameterName=Mask})
 ]
}
