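# Niagara deploy effect for the MistMylar glider, staged by emitter delays and attached to bones of the system's skeletal mesh.
# Absorb rings, beams and smoke lead into a smoky ribbon trail, then a burst of bone trails, mesh beams and smoke.
# User.VariantSwitch adds golden meshes and blends dark colors toward hot gold via array selects and lerps.
# Time-of-day, front-end and platform inputs adjust some colors, ring size and trail alpha.
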NiagaraSystem /BRCosmetics/Gadgets/Assets/VinderTech_GliderChute/Glider_MistMylar/FX/NS_Glider_MistMylar_Deploy
dependencies {
 usedDataInterfaces [
  &NiagaraDataInterfaceCurve
  &NiagaraDataInterfaceArrayFloat3
  &NiagaraDataInterfaceArrayColor
  &NiagaraDataInterfaceColorCurve
  &NiagaraDataInterfacePlatformSet
  &NiagaraDataInterfaceSpriteRendererInfo
  &NiagaraDataInterfaceArrayInt32
  &NiagaraDataInterfaceSkeletalMesh
 ]
 usedDynamicInputs []
 usedModules [
  &self:SystemSpawnScript.SystemScriptSource.SystemScriptGraph.NiagaraNodeAssignment_0.SetVariables_#1
  &SystemState
 ]
}
emitters {
 Absorb_Beam {
  emitterUpdateScript [
   &EmitterState(Life Cycle Mode=::NewEnumerator1, Loop Behavior=::NewEnumerator1, Loop Delay=0.15, Loop Duration=0.5, UseLoopDelay=true)
   &SpawnRate(
    SpawnRate=&FloatFromCurve(
     FloatCurve=DataInterface() {propertyValues {Curve {keys [
         {arriveTangent=1.5797, interpMode=RCIM_Cubic, leaveTangent=1.5797, tangentMode=RCTM_User, time=0.1, value=0}
         {arriveTangent=2.0076, interpMode=RCIM_Cubic, leaveTangent=2.0076, tangentMode=RCTM_User, time=0.6, value=1}
        ]}}}
     CurveIndex=$Emitter.NormalizedLoopAge
     Scale Curve=20
    )
    Use Spawn Probability=false
   )
  ]
  particleSpawnScript [
   &InitializeParticle(
    Lifetime=&Multiply_Float(
     A=0.4
     B=&FloatFromCurve(FloatCurve=DataInterface() {propertyValues {Curve {keys [<at 0 1>,<at 1 0.4>]}}}, CurveIndex=$Emitter.NormalizedLoopAge)
    )
    Mesh Scale Max=<1 1 1>
    Mesh Scale Min=<0.8 0.8 0.8>
    Mesh Scale Mode=::NewEnumerator2
    UsePositionOffset=false
    Write Material Random=true
   )
   &InitialMeshOrientation(
    Rotation=&RandomRangeVector(Minimum=<-1 -1 -1>, Maximum=<1 1 1>)
    Use Orientation Vector=true
    Use Rotation Vector=true
   )
  ]
  particleUpdateScript [
   &ParticleState()
   &ScaleMeshSize(Scale Factor=&VectorFromFloat(Value=1.1))
   &SkeletalMeshLocation(Bone Sampling Mode=::NewEnumerator4, MaskSpawnGroup=false, Skeletal Mesh=$System.Skeletal Mesh)
   &SolveForcesAndVelocity(Clamp Velocity=false, Limit Acceleration=false)
   &DynamicMaterialParameters(
    Index 0 Param 1=&FloatFromCurve(FloatCurve=DataInterface() {propertyValues {Curve {keys [<at 0 -0.2>,<at 1 1.2>]}}})
    Index 0 Param 2=1
    Index 0 Param 3=&Multiply_Float(
     A=&RandomRangeFloat(Minimum=1.7, Maximum=1.7)
     B=&FloatFromCurve(FloatCurve=DataInterface() {propertyValues {Curve {keys [<at 0 1>,<at 1 0.5>]}}})
    )
    Module.Param0WriteEnabled=true
    Module.Param1WriteEnabled=true
    Module.Param2WriteEnabled=true
    Module.Param3WriteEnabled=true
   )
   &ScaleColor(
    Scale Alpha=&FloatFromCurve(
     FloatCurve=DataInterface() {propertyValues {Curve {keys [<at 0 0 RCIM_Cubic>,<at 0.2 1 RCIM_Cubic>,<at 1 0 RCIM_Cubic>]}}}
     Scale Curve=&RandomRangeFloat(Minimum=1, Maximum=2)
    )
    Scale RGB=&MakeVectorFromLinearColorRGB(Color=<rgba 0.01 0.01 0.01 1>)
    ScaleA=true
    ScaleRGB=true
   )
   /SetVariables(Particles.Color_Golden:LinearColor=&MakeLinearColorFromVectorAndFloat(
     "Vector (RGB)"=&MakeVectorFromLinearColorRGB(Color=<rgba 1.5 0.7 0.05 1>)
     "Float (Alpha)"=&FloatFromCurve(
      FloatCurve=DataInterface() {propertyValues {Curve {keys [<at 0 0 RCIM_Cubic>,<at 0.2 1 RCIM_Cubic>,<at 1 0 RCIM_Cubic>]}}}
      Scale Curve=&RandomRangeFloat(Minimum=1, Maximum=1)
     )
    )) {isSetParameters=true}
   /SetVariables(Particles.Size_Golden:Vector3f=&SelectVectorFromArray(
     Direct Array Index=&MakeIntFromFloat(Float=$User.VariantSwitch)
     Vector Selection Array=DataInterface() {propertyValues {FloatData [<0 0 0>, <1 1 1>]}}
    )) {isSetParameters=true}
  ]
  rendererClasses [&NiagaraMeshRendererProperties]
  renderers [
   {rendererClass=&NiagaraMeshRendererProperties, rendererIndex=0}
   {rendererClass=&NiagaraMeshRendererProperties, rendererIndex=1}
  ]
 }
 Absorb_Ring {
  emitterUpdateScript [
   &EmitterState(Life Cycle Mode=::NewEnumerator1, Loop Behavior=::NewEnumerator1, Loop Duration=0.8, UseLoopDelay=true)
   &SpawnBurst_Instantaneous(Use Spawn Probability=false, UseLoopCountLimit=false)
  ]
  particleSpawnScript [
   &InitializeParticle(Color=<rgba 0.02 0.02 0.02 1>, Lifetime=0.8, Mesh Scale Mode=::NewEnumerator3, UsePositionOffset=false, Write Material Random=true)
   &InitialMeshOrientation(
    Mesh Orientation Mode=::NewEnumerator2
    Rotation=&RandomRangeVector(Minimum=<0 0 0>, Maximum=<1 0 0>)
    Use Rotation Vector=true
   )
  ]
  particleUpdateScript [
   &ParticleState()
   &SkeletalMeshLocation(Bone Sampling Mode=::NewEnumerator4, MaskSpawnGroup=false, Orientation Sampling=::NewEnumerator1, Skeletal Mesh=$System.Skeletal Mesh)
   /SetVariables(Particles.Size_Golden:Vector3f=&Multiply_Vector(
     A=&Multiply_VectorByFloat(
      Vector=<1 1 1>
      Float=&FloatFromCurve(
       FloatCurve=DataInterface() {propertyValues {Curve {keys [
           {arriveTangent=-4.6038, interpMode=RCIM_Cubic, leaveTangent=-4.6038, tangentMode=RCTM_User, time=0, value=7.5}
           {arriveTangent=-10.6714, interpMode=RCIM_Cubic, leaveTangent=-10.6714, tangentMode=RCTM_User, time=0.4, value=3.5}
           {arriveTangent=1.0764, interpMode=RCIM_Cubic, leaveTangent=1.0764, tangentMode=RCTM_User, time=1, value=0}
          ]}}}
       Scale Curve=&FN_FrontEndFloatModifier(Gameplay=1, Front End=0.8)
      )
     )
     B=&SelectVectorFromArray(
      Direct Array Index=&MakeIntFromFloat(Float=$User.VariantSwitch)
      Vector Selection Array=DataInterface() {propertyValues {FloatData [<0 0 0>, <1 1 1>]}}
     )
    )) {isSetParameters=true}
   &ScaleMeshSize(Scale Factor=&Multiply_VectorByFloat(
     Vector=<1 1 1>
     Float=&FloatFromCurve(
      FloatCurve=DataInterface() {propertyValues {Curve {keys [
          {arriveTangent=-4.6038, interpMode=RCIM_Cubic, leaveTangent=-4.6038, tangentMode=RCTM_User, time=0, value=7.5}
          {arriveTangent=-10.6714, interpMode=RCIM_Cubic, leaveTangent=-10.6714, tangentMode=RCTM_User, time=0.4, value=3.5}
          {arriveTangent=1.0764, interpMode=RCIM_Cubic, leaveTangent=1.0764, tangentMode=RCTM_User, time=1, value=0}
         ]}}}
      Scale Curve=&FN_FrontEndFloatModifier(Gameplay=1, Front End=0.8)
     )
    ))
   &SolveForcesAndVelocity(Clamp Velocity=false, Limit Acceleration=false)
   &ScaleColor(
    Scale Alpha=&FloatFromCurve(
     FloatCurve=DataInterface() {propertyValues {Curve {keys [
         {arriveTangent=0.3287, interpMode=RCIM_Cubic, leaveTangent=0.3287, tangentMode=RCTM_User, time=0, value=0}
         {arriveTangent=0.0273, interpMode=RCIM_Cubic, leaveTangent=0.0273, tangentMode=RCTM_User, time=0.6, value=1}
         {arriveTangent=-1.267, interpMode=RCIM_Cubic, leaveTangent=-1.267, tangentMode=RCTM_User, time=0.9, value=0.8}
         {arriveTangent=-4.9486, interpMode=RCIM_Cubic, leaveTangent=-4.9486, tangentMode=RCTM_User, time=1, value=0}
        ]}}}
     Scale Curve=1.6
    )
    Scale RGB=&MakeVectorFromLinearColorRGB(Color=&FNTimeOfDayColor(
      Night Color=<rgba 0.018 0.021 0.03 1>
      Morning Color=<rgba 0.05 0.05 0.05 1>
      Day Color=<rgba 0.05 0.05 0.05 1>
      Evening Color=<rgba 0.05 0.05 0.05 1>
     ))
    ScaleA=true
    ScaleRGB=true
   )
   /SetVariables(Particles.Color_Golden:LinearColor=&MakeLinearColorFromVectorAndFloat(
     "Vector (RGB)"=&MakeVectorFromLinearColorRGB(Color=&Multiply_LinearColorByFloat(
       Linear Color=<rgba 0.3698 0.1403 0.0039 1>
       Float=&FNTimeOfDayScalar(Night Scale=0.6, Morning Scale=1, Evening Scale=1)
      ))
     "Float (Alpha)"=&FloatFromCurve(
      FloatCurve=DataInterface() {propertyValues {Curve {keys [
          {arriveTangent=0.3287, interpMode=RCIM_Cubic, leaveTangent=0.3287, tangentMode=RCTM_User, time=0.1, value=0}
          {arriveTangent=0.9368, interpMode=RCIM_Cubic, leaveTangent=0.9368, tangentMode=RCTM_User, time=0.8, value=1}
          {arriveTangent=-1.4842, interpMode=RCIM_Cubic, leaveTangent=-1.4842, tangentMode=RCTM_User, time=1, value=0}
         ]}}}
      Scale Curve=0.15
     )
    )) {isSetParameters=true}
  ]
  rendererClasses [&NiagaraMeshRendererProperties]
  renderers [
   {rendererClass=&NiagaraMeshRendererProperties, rendererIndex=0}
   {rendererClass=&NiagaraMeshRendererProperties, rendererIndex=1}
  ]
 }
 Absorb_Smoke {
  emitterUpdateScript [
   &EmitterState(
    Inactive Response=::NewEnumerator1
    Life Cycle Mode=::NewEnumerator1
    Loop Behavior=::NewEnumerator1
    Loop Delay=0.2
    Loop Duration=0.4
    UseLoopDelay=true
   )
   &SpawnRate(SpawnRate=25, Use Spawn Probability=false)
  ]
  particleSpawnScript [&InitializeParticle(
    Color=&SelectLinearColorFromArray(
     Array Sampling Mode=::NewEnumerator1
     Direct Array Index=&MakeIntFromFloat(Float=$User.VariantSwitch)
     Color Selection Array=DataInterface() {propertyValues {ColorData [<rgba 0.01 0.01 0.01 1.5>, <rgba 0.05 0.05 0.05 1.5>]}}
     Random Seed=?
    )
    Lifetime=&Multiply_Float(
     A=&RandomRangeFloat(Minimum=0.3, Maximum=0.3)
     B=&FloatFromCurve(FloatCurve=DataInterface() {propertyValues {Curve {keys [<at 0 1>,<at 1 0.3>]}}}, CurveIndex=$Emitter.NormalizedLoopAge)
    )
    Sprite Rotation Mode=::NewEnumerator2
    Sprite Size Mode=::NewEnumerator1
    Uniform Sprite Size Max=120
    Uniform Sprite Size Min=100
    UsePositionOffset=false
    Write Material Random=true
   )]
  particleUpdateScript [
   &ParticleState()
   &SpriteRotationRate(Rotation Rate=&FloatFromCurve(
     FloatCurve=DataInterface() {propertyValues {Curve {keys [{interpMode=RCIM_Cubic, tangentMode=RCTM_Break, time=0, value=1}, {interpMode=RCIM_Cubic, tangentMode=RCTM_Break, time=1, value=0.489}]}}}
     Scale Curve=&Lerp_Float(A=-150, B=150, Alpha=$Particles.Initial.MaterialRandom)
    ))
   &ScaleSpriteSize(Uniform Curve Sprite Scale=DataInterface() {propertyValues {Curve {keys [{tangentMode=RCTM_User, time=0, value=1}, {tangentMode=RCTM_User, time=1, value=0.45}]}}})
   &SkeletalMeshLocation(
    Bone Sampling Mode=::NewEnumerator4
    MaskSpawnGroup=false
    Sampled Position Offset=&RandomRangeVector(
     Minimum=&AddVectorToPosition(
      Position=<70 70 70>
      Vector=&VectorFromFloat(Value=&FloatFromCurve(FloatCurve=DataInterface() {propertyValues {Curve {keys [<at 0 0>,<at 1 -70>]}}}))
     )
     Maximum=&AddVectorToPosition(
      Position=<-70 -70 -70>
      Vector=&VectorFromFloat(Value=&FloatFromCurve(FloatCurve=DataInterface() {propertyValues {Curve {keys [<at 0 0>,<at 1 70>]}}}))
     )
    )
    Skeletal Mesh=$System.Skeletal Mesh
   )
   &SolveForcesAndVelocity(Clamp Velocity=false, Limit Acceleration=false)
   &ScaleColor(
    Scale Alpha=&FloatFromCurve(FloatCurve=DataInterface() {propertyValues {Curve {keys [
         {interpMode=RCIM_Cubic, tangentMode=RCTM_User, time=0, value=0}
         {interpMode=RCIM_Cubic, tangentMode=RCTM_User, time=0.3, value=1}
         {tangentMode=RCTM_User, time=1, value=0}
        ]}}})
    ScaleA=true
    ScaleRGB=true
   )
  ]
  rendererClasses [&NiagaraSpriteRendererProperties]
  renderers [{rendererClass=&NiagaraSpriteRendererProperties}]
 }
 Burat_smoke {
  emitterUpdateScript [
   &EmitterState(
    Inactive Response=::NewEnumerator1
    Life Cycle Mode=::NewEnumerator1
    Loop Behavior=::NewEnumerator1
    Loop Delay=2.1
    Loop Duration=10
    UseLoopDelay=true
   )
   &SpawnBurst_Instantaneous(Spawn Count=25, Use Spawn Probability=false, UseLoopCountLimit=false)
  ]
  particleSpawnScript [
   &InitializeParticle(
    Color Maximum=<rgba 0.4219 0.4219 0.4219 1>
    Color Minimum=<rgba 0.1406 0.1406 0.1406 1>
    Color Mode=::NewEnumerator2
    Lifetime Max=1.3
    Lifetime Min=0.8
    Lifetime Mode=::NewEnumerator1
    Sprite Rotation Mode=::NewEnumerator2
    Sprite Size Mode=::NewEnumerator1
    Uniform Sprite Size Max=120
    Uniform Sprite Size Min=100
    UsePositionOffset=false
    Write Material Random=true
   )
   &CameraOffset(Camera Offset Amount=100)
  ]
  particleUpdateScript [
   &ParticleState()
   &SpriteRotationRate(Rotation Rate=&RandomRangeFloat(Minimum=-100, Maximum=100))
   &ScaleSpriteSize(Uniform Curve Sprite Scale=DataInterface() {propertyValues {Curve {keys [{interpMode=RCIM_Cubic, tangentMode=RCTM_User, time=0, value=1.05}, {interpMode=RCIM_Cubic, tangentMode=RCTM_User, time=1, value=2}]}}})
   &SkeletalMeshLocation(
    Bone Sampling Mode=::NewEnumerator4
    MaskSpawnGroup=false
    Sampled Position Offset=&MakeVector(
     X=&FloatFromCurve(
      FloatCurve=DataInterface() {propertyValues {Curve {keys [
          {arriveTangent=232.0188, interpMode=RCIM_Cubic, leaveTangent=232.0188, tangentMode=RCTM_User, time=0, value=1}
          {arriveTangent=249, interpMode=RCIM_Cubic, leaveTangent=249, time=0.3, value=125}
          <at 1 250 RCIM_Cubic>
         ]}}}
      Scale Curve=&RandomRangeFloat(Minimum=-1, Maximum=1)
     )
     Y=&FloatFromCurve(
      FloatCurve=DataInterface() {propertyValues {Curve {keys [<at 0 1 RCIM_Cubic>, {arriveTangent=329, interpMode=RCIM_Cubic, leaveTangent=329, time=0.3, value=187.5}, <at 1 330 RCIM_Cubic>]}}}
      Scale Curve=&RandomRangeFloat(Minimum=0.65, Maximum=1)
     )
     Z=&FloatFromCurve(
      FloatCurve=DataInterface() {propertyValues {Curve {keys [
          {arriveTangent=472.8612, interpMode=RCIM_Cubic, leaveTangent=472.8612, tangentMode=RCTM_User, time=0, value=1}
          {arriveTangent=219, interpMode=RCIM_Cubic, leaveTangent=219, time=0.3, value=125}
          <at 1 220 RCIM_Cubic>
         ]}}}
      Scale Curve=&RandomRangeFloat(Minimum=-0.2, Maximum=1)
     )
    )
    Skeletal Mesh=$System.Skeletal Mesh
   )
   &ScaleColor(
    Scale Alpha=&Multiply_Float(
     A=&FloatFromCurve(
      FloatCurve=DataInterface() {propertyValues {Curve {keys [
          {arriveTangent=10.2356, interpMode=RCIM_Cubic, leaveTangent=10.2356, tangentMode=RCTM_User, time=0, value=0}
          <at 0.3 1>
          <at 0.7 0.4>
          <at 1 0>
         ]}}}
      Scale Curve=&RandomRangeFloat(Minimum=0.6, Maximum=1.65)
     )
     B=1
    )
    Scale RGB=/Niagara/DynamicInputs/Math/LerpMultipleVectors(
     Alpha:NiagaraFloat=&Divide_Float(A=$User.VariantSwitch, B=2)
     V 1:Vector3f=&MakeVectorFromLinearColorRGB(Color=&ColorFromCurve(ColorCurve=DataInterface() {propertyValues {
         AlphaCurve {keys []}
         BlueCurve {keys [{arriveTangent=-331.6383, interpMode=RCIM_Cubic, time=0, value=0.03}, <at 0.5442 0.005 RCIM_Cubic>]}
         GreenCurve {keys [{arriveTangent=-87.3695, interpMode=RCIM_Cubic, time=0, value=0.03}, <at 0.5442 0.005 RCIM_Cubic>]}
         RedCurve {keys [<at 0 0.03 RCIM_Cubic>,<at 0.5442 0.005 RCIM_Cubic>]}
        }}))
     V 2:Vector3f=&MakeVectorFromLinearColorRGB(Color=&ColorFromCurve(ColorCurve=DataInterface() {propertyValues {
         AlphaCurve {keys []}
         BlueCurve {keys [
           {arriveTangent=-4.6751, time=0.1217, value=0.03}
           <at 0.2471 0.0599>
           {arriveTangent=-0.2522, leaveTangent=-0.2522, time=0.3307, value=0.1389}
           <at 0.6322 0.01>
          ]}
         GreenCurve {keys [
           {arriveTangent=-14.5607, time=0.1217, value=0.03}
           <at 0.2471 0.3033>
           {arriveTangent=-3.3642, leaveTangent=-3.3642, time=0.3307, value=0.1542}
           <at 0.6322 0.01>
          ]}
         RedCurve {keys [
           {arriveTangent=-27.0666, time=0.1217, value=0.03}
           <at 0.2471 0.5>
           {arriveTangent=-6.5976, leaveTangent=-6.5976, time=0.3307, value=0.1667}
           <at 0.6322 0.01>
          ]}
        }}))
     V 3:Vector3f=?
     V 4:Vector3f=?
     V 5:Vector3f=?
     Vector Count:ENiagara_MultipleLerpCount=::NewEnumerator3
    )
    ScaleA=true
    ScaleRGB=true
   )
   &SubUVAnimation(
    End Frame Range Override=3
    Evaluation Type=::NewEnumerator0
    "Sprite Renderer "=DataInterface() {propertyValues {SpriteRenderer=&self:Burat_smoke_5.NiagaraSpriteRendererProperties_1}}
    SubUV Animation Mode:ENiagara_SubUVLookupModeV2=::NewEnumerator2
    UseEndFrame=true
    UseStartFrame=true
   )
   &SolveForcesAndVelocity(Clamp Velocity=false, Limit Acceleration=false)
  ]
  rendererClasses [&NiagaraSpriteRendererProperties]
  renderers [{rendererClass=&NiagaraSpriteRendererProperties}]
 }
 Burst_Beam {
  emitterUpdateScript [
   &EmitterState(Life Cycle Mode=::NewEnumerator1, Loop Behavior=::NewEnumerator1, Loop Delay=2.2, UseLoopDelay=true)
   &SpawnBurst_Instantaneous(Spawn Count=8, Use Spawn Probability=false, UseLoopCountLimit=false)
  ]
  particleSpawnScript [
   &PartitionParticles(Number Of Partitions=8)
   &InitializeParticle(
    Color=&FNTimeOfDayColor(Night Color=<rgba 0.6 0.7 1 1>, Morning Color=<rgba 1 1 1 1>, Day Color=<rgba 1 1 1 1>, Evening Color=<rgba 1 1 1 1>)
    Lifetime Max=1.2
    Lifetime Min=0.7
    Lifetime Mode=::NewEnumerator1
    Mesh Scale=&SelectVectorFromArray(
     Direct Array Index=$Particles.PartitionParticles.Partition
     Vector Selection Array=DataInterface() {propertyValues {FloatData [<1.2 1.2 2.3>, <1.5 1.5 2.5>, <1 1 2.8>, <1 1 2.3>, <1 1 2.2>, <1.2 1.2 2.6>, <1.2 1.2 2.8>, <1 1 2.3>]}}
    )
    Mesh Scale Mode=::NewEnumerator4
    UsePositionOffset=false
    Write Material Random=true
   )
   &InitialMeshOrientation(
    Mesh Orientation Mode=::NewEnumerator2
    Rotation=&SelectVectorFromArray(
     Direct Array Index=$Particles.PartitionParticles.Partition
     Vector Selection Array=DataInterface() {propertyValues {FloatData [<0 -0.25 0>, <0 -0.25 0>, <0 -0.25 0>, <0 -0.25 0>, <0 0.25 0>, <0 0.25 0>, <0 0.25 0>, <0 0.25 0>]}}
    )
    Use Rotation Vector=true
   )
   &InitialMeshOrientation(
    Mesh Orientation Mode=::NewEnumerator2
    Rotation=&SelectVectorFromArray(
     Direct Array Index=$Particles.PartitionParticles.Partition
     Vector Selection Array=DataInterface() {propertyValues {FloatData [<0 0.1 -0.15>, <0 0.04 -0.12>, <0 -0.04 -0.08>, <0 -0.07 -0.1>, <0 -0.08 0.15>, <0 -0.02 0.08>, <0 0.04 0.1>, <0 0.12 0>]}}
    )
    Use Rotation Vector=true
   ) {module=InitialMeshOrientation001}
   &InitialMeshOrientation(Mesh Orientation Mode=::NewEnumerator2, Rotation=<0.05 0 0>, Use Rotation Vector=true) {module=InitialMeshOrientation002}
   &DynamicMaterialParameters(
    Index 0 Param 1=&Multiply_Float(
     A=&RandomRangeFloat(Minimum=-0.4, Maximum=-1)
     B=&FloatFromCurve(FloatCurve=DataInterface() {propertyValues {Curve {keys [<at 0 1>,<at 1 0>]}}})
    )
    Index 0 Param 3=&RandomRangeFloat(Minimum=0.5, Maximum=1)
    Index 0 Param 4=1
    Index 1 Param 1=1
    Index 1 Param 2=1
    Module.Index0Param0Write=true
    Module.Index0Param1Write=true
    Module.Index0Param2Write=true
    Module.Index0Param3Write=true
    Module.Index1Param0Write=true
    Module.Index1Param1Write=true
    Module.Index1Param2Write=true
    Module.Index1Param3Write=true
    Write Parameter Index 1=true
   )
  ]
  particleUpdateScript [
   &ParticleState()
   &ScaleMeshSize(Scale Factor=&MakeVector(
     X=1
     Y=1
     Z=&FloatFromCurve(
      FloatCurve=DataInterface() {propertyValues {Curve {keys [
          {arriveTangent=1.7446, interpMode=RCIM_Cubic, leaveTangent=1.7446, tangentMode=RCTM_User, time=0, value=0.55}
          {arriveTangent=0.1777, interpMode=RCIM_Cubic, leaveTangent=0.1777, tangentMode=RCTM_User, time=1, value=1.4}
         ]}}}
      Scale Curve=2
     )
    ))
   &SkeletalMeshLocation(Bone Sampling Mode=::NewEnumerator4, MaskSpawnGroup=false, Skeletal Mesh=$System.Skeletal Mesh)
   &ScaleColor(
    Scale Alpha=&FloatFromCurve(
     FloatCurve=DataInterface() {propertyValues {Curve {keys [
         {arriveTangent=2.8552, interpMode=RCIM_Cubic, leaveTangent=2.8552, tangentMode=RCTM_User, time=0, value=0}
         {arriveTangent=0.0143, interpMode=RCIM_Cubic, leaveTangent=0.0143, tangentMode=RCTM_User, time=0.1, value=1}
         {arriveTangent=-2.3539, interpMode=RCIM_Cubic, leaveTangent=-2.3539, tangentMode=RCTM_User, time=0.4, value=0.55}
         {arriveTangent=-0.0546, interpMode=RCIM_Cubic, leaveTangent=-0.0546, tangentMode=RCTM_User, time=1, value=0}
        ]}}}
     Scale Curve=1.2
    )
    Scale RGB=&LerpMultipleVectors(
     Alpha:NiagaraFloat=&Divide_Float(A=$User.VariantSwitch, B=2)
     V 1:Vector3f=&MakeVectorFromLinearColorRGB(Color=&ColorFromCurve(ColorCurve=DataInterface() {propertyValues {
         AlphaCurve {keys []}
         BlueCurve {keys [<at 0.0663 0.02 RCIM_Cubic>]}
         GreenCurve {keys [<at 0.0663 0.02 RCIM_Cubic>]}
         RedCurve {keys [<at 0.0663 0.02 RCIM_Cubic>]}
        }}))
     V 2:Vector3f=&MakeVectorFromLinearColorRGB(Color=&ColorFromCurve(ColorCurve=DataInterface() {propertyValues {
         AlphaCurve {keys []}
         BlueCurve {keys [<at 0.0455 0.1738>,<at 0.2023 0.0122>,<at 0.4068 0.01 RCIM_Cubic>]}
         GreenCurve {keys [<at 0.0455 1.0001>,<at 0.2023 0.1673>,<at 0.4068 0.01 RCIM_Cubic>]}
         RedCurve {keys [<at 0.0455 2>,<at 0.2023 0.4219>,<at 0.4068 0.01 RCIM_Cubic>]}
        }}))
     V 3:Vector3f=?
     V 4:Vector3f=?
     V 5:Vector3f=?
     Vector Count:ENiagara_MultipleLerpCount=::NewEnumerator3
    )
    ScaleA=true
    ScaleRGB=true
   )
  ]
  rendererClasses [&NiagaraMeshRendererProperties]
  renderers [{rendererClass=&NiagaraMeshRendererProperties, rendererIndex=0}]
 }
 Burst_Trail {
  emitterUpdateScript [
   &EmitterState(Life Cycle Mode=::NewEnumerator1, Loop Behavior=::NewEnumerator1, Loop Delay=1.9, Loop Duration=0.5, UseLoopDelay=true)
   &SpawnRate(SpawnRate=200, Use Spawn Probability=false)
  ]
  particleSpawnScript [
   &PartitionParticles(Number Of Partitions=10)
   &InitializeParticle(Color=<rgba 0.03 0.03 0.03 1.5>, Ribbon Width=30, Ribbon Width Mode=::NewEnumerator1, UsePositionOffset=false, Write Material Random=true)
   &SkeletalMeshLocation(
    Bone / Socket Index=&SelectIntFromArray(
     Direct Array Index=$Particles.PartitionParticles.Partition
     Int Selection Array=DataInterface() {propertyValues {IntData [3,4,5,6,7,8,9,10]}}
    )
    Bone Sampling Mode=::NewEnumerator4
    MaskSpawnGroup=false
    Sampled Position Offset Coordinate Space:ENiagaraSkelMeshTransforms=::NewEnumerator2
    Skeletal Mesh=$System.Skeletal Mesh
   )
  ]
  particleUpdateScript [
   &ParticleState()
   &InheritSourceMovement(ApplyForce=false, ApplyPosition=true, ApplyPositionFalloff=false)
   &ScaleColor(
    Scale Alpha=&FloatFromCurve(
     FloatCurve=DataInterface() {propertyValues {Curve {keys [
         {arriveTangent=-8.8443, interpMode=RCIM_Cubic, leaveTangent=-8.8443, tangentMode=RCTM_User, time=0, value=2}
         {arriveTangent=-1.8244, interpMode=RCIM_Cubic, leaveTangent=-1.8244, tangentMode=RCTM_User, time=0.2, value=0.5}
         {arriveTangent=0.0525, interpMode=RCIM_Cubic, leaveTangent=0.0525, tangentMode=RCTM_User, time=1, value=0}
        ]}}}
     Scale Curve=&Multiply_Float(
      A=0.5
      B=&FloatFromCurve(FloatCurve=DataInterface() {propertyValues {Curve {keys [<at 0 1>,<at 2.3 1>,<at 3 0>]}}}, CurveIndex=$Emitter.Age)
     )
    )
    Scale RGB=&LerpMultipleVectors(
     Alpha:NiagaraFloat=&Divide_Float(A=$User.VariantSwitch, B=2)
     V 1:Vector3f=&MakeVectorFromLinearColorRGB(Color=&ColorFromCurve(ColorCurve=DataInterface() {propertyValues {
         AlphaCurve {keys []}
         BlueCurve {keys [<at 0.0663 0.06 RCIM_Cubic>,<at 0.4163 0.01 RCIM_Cubic>]}
         GreenCurve {keys [<at 0.0663 0.06 RCIM_Cubic>,<at 0.4163 0.01 RCIM_Cubic>]}
         RedCurve {keys [<at 0.0663 0.06 RCIM_Cubic>,<at 0.4163 0.01 RCIM_Cubic>]}
        }}))
     V 2:Vector3f=&MakeVectorFromLinearColorRGB(Color=&ColorFromCurve(ColorCurve=DataInterface() {propertyValues {
         AlphaCurve {keys []}
         BlueCurve {keys [<at 0.0403 2.2831 RCIM_Cubic>,<at 0.1619 0.01 RCIM_Cubic>]}
         GreenCurve {keys [<at 0.0403 14.4445 RCIM_Cubic>,<at 0.1619 0.01 RCIM_Cubic>]}
         RedCurve {keys [<at 0.0403 20 RCIM_Cubic>,<at 0.1619 0.01 RCIM_Cubic>]}
        }}))
     V 3:Vector3f=?
     V 4:Vector3f=?
     V 5:Vector3f=?
     Vector Count:ENiagara_MultipleLerpCount=::NewEnumerator3
    )
    ScaleA=true
    ScaleRGB=true
   )
   &ScaleRibbonWidth(Ribbon Width Scale=&FloatFromCurve(FloatCurve=DataInterface() {propertyValues {Curve {keys [{arriveTangent=-1.0264, time=-0.2, value=1}, {arriveTangent=-0.3983, leaveTangent=-0.3983, time=1, value=0}]}}}))
   &SolveForcesAndVelocity(Clamp Velocity=false, Limit Acceleration=false)
  ]
  rendererClasses [&NiagaraRibbonRendererProperties]
  renderers [{rendererClass=&NiagaraRibbonRendererProperties}]
 }
 Ribbon_Smoke {
  emitterUpdateScript [
   &EmitterState(
    Inactive Response=::NewEnumerator1
    Life Cycle Mode=::NewEnumerator1
    Loop Behavior=::NewEnumerator1
    Loop Delay=1
    Loop Duration=1.3
    UseLoopDelay=true
   )
   &SpawnRate(
    SpawnRate=&FloatFromCurve(
     FloatCurve=DataInterface() {propertyValues {Curve {keys [<at 0 1>,<at 0.5 1>,<at 0.8 0>]}}}
     CurveIndex=$Emitter.NormalizedLoopAge
     Scale Curve=30
    )
    Use Spawn Probability=false
   )
  ]
  particleSpawnScript [
   &InitializeParticle(
    Color=&FNTimeOfDayColor(
     Night Color=<rgba 0.5521 0.6641 1 0.8>
     Morning Color=<rgba 0.7 0.7 0.7 0.8>
     Day Color=<rgba 0.7 0.7 0.7 0.8>
     Evening Color=<rgba 0.7 0.7 0.7 0.8>
    )
    Lifetime Max=1
    Lifetime Min=0.8
    Lifetime Mode=::NewEnumerator1
    Sprite Rotation Mode=::NewEnumerator2
    Sprite Size Mode=::NewEnumerator1
    Uniform Sprite Size Max=150
    Uniform Sprite Size Min=100
    UsePositionOffset=false
    Write Material Random=true
   )
   &SkeletalMeshLocation(
    Bone / Socket Index=1
    Bone Sampling Mode=::NewEnumerator4
    MaskSpawnGroup=false
    Sampled Position Offset=<-80 -30 0>
    Skeletal Mesh=$System.Skeletal Mesh
   )
   &AddVelocity(Velocity=&MakeVector(
     X=0
     Y=&Multiply_Float(
      A=500
      B=&FloatFromCurve(FloatCurve=DataInterface() {propertyValues {Curve {keys [<at 0 1>,<at 0.7 0.05>]}}})
     )
     Z=&Multiply_Float(
      A=-100
      B=&FloatFromCurve(FloatCurve=DataInterface() {propertyValues {Curve {keys [<at 0 1>,<at 0.7 0.05>]}}})
     )
    ))
  ]
  particleUpdateScript [
   &ParticleState()
   &ScaleSpriteSize(
    Uniform Curve Scale=1.8
    Uniform Curve Sprite Scale=DataInterface() {propertyValues {Curve {keys [
        {arriveTangent=1.6793, interpMode=RCIM_Cubic, leaveTangent=1.6793, tangentMode=RCTM_User, time=0, value=0.6}
        {arriveTangent=0.0143, interpMode=RCIM_Cubic, leaveTangent=0.0143, tangentMode=RCTM_User, time=0.5, value=1}
        <at 1 0.6>
       ]}}}
   )
   &SpriteRotationRate(Rotation Rate=&RandomRangeFloat(Minimum=-100, Maximum=100))
   &SubUVAnimation(
    "Sprite Renderer "=DataInterface() {propertyValues {SpriteRenderer=&self:Ribbon_Smoke_10.NiagaraSpriteRendererProperties_0}}
    UseEndFrame=true
    UseStartFrame=true
   )
   &ScaleColor(
    Scale Alpha=&FloatFromCurve(
     FloatCurve=DataInterface() {propertyValues {Curve {keys [
         <at 0 0 RCIM_Cubic>
         {arriveTangent=1.2, interpMode=RCIM_Cubic, leaveTangent=1.2, time=0.2, value=1}
         {arriveTangent=-3.1331, interpMode=RCIM_Cubic, leaveTangent=-3.1331, tangentMode=RCTM_User, time=0.5, value=0.6}
         <at 1 0 RCIM_Cubic>
        ]}}}
     Scale Curve=0.8
    )
    Scale RGB=&MakeVectorFromLinearColorRGB(Color=&Multiply_LinearColorByFloat(
      Linear Color=&ColorFromCurve(ColorCurve=DataInterface() {propertyValues {
         AlphaCurve {keys []}
         BlueCurve {keys [{arriveTangent=-331.6383, interpMode=RCIM_Cubic, time=0, value=0.1}, <at 0.7598 0.005 RCIM_Cubic>]}
         GreenCurve {keys [{arriveTangent=-87.3695, interpMode=RCIM_Cubic, time=0, value=0.1}, <at 0.7598 0.005 RCIM_Cubic>]}
         RedCurve {keys [<at 0 0.1 RCIM_Cubic>,<at 0.7598 0.005 RCIM_Cubic>]}
        }})
      Float=1
     ))
    ScaleA=true
    ScaleRGB=true
   )
   &DynamicMaterialParameters(
    Index 0 Param 1=&FloatFromCurve(
     FloatCurve=DataInterface() {propertyValues {Curve {keys [
         {interpMode=RCIM_Cubic, tangentMode=RCTM_User, time=0, value=0}
         <at 0.2 0.55>
         {arriveTangent=-1.0223, interpMode=RCIM_Cubic, leaveTangent=-1.0223, tangentMode=RCTM_User, time=1, value=1}
        ]}}}
     Scale Curve=&RandomRangeFloat(Minimum=3, Maximum=3)
    )
    Module.Index0Param0Write=true
    Module.Index0Param1Write=true
    Module.Index0Param2Write=true
    Module.Index0Param3Write=true
   )
   &InheritSourceMovement(ApplyForce=false, ApplyPosition=true, ApplyPositionFalloff=false)
   &Drag(Use Linear Drag=true, Use Rotational Drag=true)
   &SolveForcesAndVelocity(Clamp Velocity=false, Limit Acceleration=false)
  ]
  rendererClasses [&NiagaraSpriteRendererProperties]
  renderers [{rendererClass=&NiagaraSpriteRendererProperties}]
 }
 Ribbon_Smoke001 {
  emitterUpdateScript [
   &EmitterState(
    Inactive Response=::NewEnumerator1
    Life Cycle Mode=::NewEnumerator1
    Loop Behavior=::NewEnumerator1
    Loop Delay=1
    Loop Duration=1.3
    UseLoopDelay=true
   )
   &SpawnRate(
    SpawnRate=&FloatFromCurve(
     FloatCurve=DataInterface() {propertyValues {Curve {keys [<at 0 1>,<at 0.5 1>,<at 0.8 0>]}}}
     CurveIndex=$Emitter.NormalizedLoopAge
     Scale Curve=15
    )
    Use Spawn Probability=false
   )
  ]
  particleSpawnScript [
   &InitializeParticle(
    Lifetime Max=1
    Lifetime Min=0.8
    Lifetime Mode=::NewEnumerator1
    Sprite Rotation Mode=::NewEnumerator2
    Sprite Size Mode=::NewEnumerator1
    Uniform Sprite Size Max=200
    Uniform Sprite Size Min=150
    UsePositionOffset=false
    Write Material Random=true
   )
   &SkeletalMeshLocation(
    Bone / Socket Index=1
    Bone Sampling Mode=::NewEnumerator4
    MaskSpawnGroup=false
    Sampled Position Offset=<-80 -30 0>
    Skeletal Mesh=$System.Skeletal Mesh
   )
   &AddVelocity(Velocity=&MakeVector(
     X=0
     Y=&Multiply_Float(
      A=500
      B=&FloatFromCurve(FloatCurve=DataInterface() {propertyValues {Curve {keys [<at 0 1>,<at 0.7 0.05>]}}})
     )
     Z=&Multiply_Float(
      A=-100
      B=&FloatFromCurve(FloatCurve=DataInterface() {propertyValues {Curve {keys [<at 0 1>,<at 0.7 0.05>]}}})
     )
    ))
  ]
  particleUpdateScript [
   &ParticleState()
   &ScaleSpriteSize(
    Uniform Curve Scale=1.8
    Uniform Curve Sprite Scale=DataInterface() {propertyValues {Curve {keys [
        {arriveTangent=1.6793, interpMode=RCIM_Cubic, leaveTangent=1.6793, tangentMode=RCTM_User, time=0, value=0.6}
        {arriveTangent=0.0143, interpMode=RCIM_Cubic, leaveTangent=0.0143, tangentMode=RCTM_User, time=0.5, value=1}
        <at 1 0.6>
       ]}}}
   )
   &SpriteRotationRate(Rotation Rate=&RandomRangeFloat(Minimum=-150, Maximum=150))
   &ScaleColor(
    Scale Alpha=&FloatFromCurve(
     FloatCurve=DataInterface() {propertyValues {Curve {keys [
         <at 0 0 RCIM_Cubic>
         {arriveTangent=1, interpMode=RCIM_Cubic, leaveTangent=1, time=0.3, value=1}
         {arriveTangent=-3.1331, interpMode=RCIM_Cubic, leaveTangent=-3.1331, tangentMode=RCTM_User, time=0.6, value=0.6}
         <at 1 0 RCIM_Cubic>
        ]}}}
     Scale Curve=1.5
    )
    Scale RGB=&LerpMultipleVectors(
     Alpha:NiagaraFloat=&Divide_Float(A=$User.VariantSwitch, B=2)
     V 1:Vector3f=&Multiply_Vector(
      A=&MakeVectorFromLinearColorRGB(Color=&ColorFromCurve(ColorCurve=DataInterface() {propertyValues {
          AlphaCurve {keys []}
          BlueCurve {keys [<at 0 0.1>,<at 0.52 0.01 RCIM_Cubic>]}
          GreenCurve {keys [<at 0 0.1>,<at 0.52 0.01 RCIM_Cubic>]}
          RedCurve {keys [<at 0 0.1>,<at 0.52 0.01 RCIM_Cubic>]}
         }}))
      B=&MakeVectorFromLinearColorRGB(Color=&FNTimeOfDayColor(Night Color=<rgba 0.3 0.375 0.6 1>, Morning Color=<rgba 1 1 1 1>, Day Color=<rgba 1 1 1 1>, Evening Color=<rgba 1 1 1 1>))
     )
     V 2:Vector3f=&MakeVectorFromLinearColorRGB(Color=&ColorFromCurve(ColorCurve=DataInterface() {propertyValues {
         AlphaCurve {keys []}
         BlueCurve {keys [{arriveTangent=-0.2522, time=0.0411, value=0.1188}, <at 0.2769 0.01>]}
         GreenCurve {keys [{arriveTangent=-3.3642, time=0.0411, value=0.6025}, <at 0.2769 0.01>]}
         RedCurve {keys [{arriveTangent=-6.5976, time=0.0411, value=1}, <at 0.2769 0.01>]}
        }}))
     V 3:Vector3f=?
     V 4:Vector3f=?
     V 5:Vector3f=?
     Vector Count:ENiagara_MultipleLerpCount=::NewEnumerator3
    )
    ScaleA=true
    ScaleRGB=true
   )
   &DynamicMaterialParameters(
    Index 0 Param 1=&FloatFromCurve(
     FloatCurve=DataInterface() {propertyValues {Curve {keys [
         {interpMode=RCIM_Cubic, tangentMode=RCTM_User, time=0, value=0}
         <at 0.2 0.55>
         {arriveTangent=-1.0223, interpMode=RCIM_Cubic, leaveTangent=-1.0223, tangentMode=RCTM_User, time=1, value=1}
        ]}}}
     Scale Curve=&RandomRangeFloat(Minimum=3, Maximum=3)
    )
    Module.Index0Param0Write=true
    Module.Index0Param1Write=true
    Module.Index0Param2Write=true
    Module.Index0Param3Write=true
   ) {enabled=false}
   &InheritSourceMovement(ApplyForce=false, ApplyPosition=true, ApplyPositionFalloff=false)
   &Drag(Drag=1.2, Use Linear Drag=true, Use Rotational Drag=true)
   &SolveForcesAndVelocity(Clamp Velocity=false, Limit Acceleration=false)
  ]
  rendererClasses [&NiagaraSpriteRendererProperties]
  renderers [{rendererClass=&NiagaraSpriteRendererProperties}]
 }
 Ribbon_Trail {
  emitterUpdateScript [
   &EmitterState(Life Cycle Mode=::NewEnumerator1, Loop Behavior=::NewEnumerator1, Loop Delay=1.05, Loop Duration=0.85, UseLoopDelay=true)
   &SpawnRate(SpawnRate=30, Use Spawn Probability=false)
  ]
  particleSpawnScript [
   &InitializeParticle(
    Color=&FNTimeOfDayColor(
     Night Color=<rgba 0.05 0.0625 0.1 1>
     Morning Color=<rgba 0.02 0.02 0.02 1>
     Day Color=<rgba 0.02 0.02 0.02 1>
     Evening Color=<rgba 0.02 0.02 0.02 1>
    )
    Lifetime=0.65
    Ribbon Width=100
    Ribbon Width Mode=::NewEnumerator1
    UsePositionOffset=false
    Write Material Random=true
   )
   &SkeletalMeshLocation(
    Bone Sampling Mode=::NewEnumerator4
    MaskSpawnGroup=false
    Orientation Sampling=::NewEnumerator1
    Sampled Position Offset=<-80 -30 0>
    Skeletal Mesh=$System.Skeletal Mesh
   )
   &AddVelocity(Velocity=&MakeVector(
     X=0
     Y=&Multiply_Float(
      A=500
      B=&FloatFromCurve(FloatCurve=DataInterface() {propertyValues {Curve {keys [<at 0 1>,<at 0.2 0.125>]}}})
     )
     Z=&Multiply_Float(
      A=-100
      B=&FloatFromCurve(FloatCurve=DataInterface() {propertyValues {Curve {keys [<at 0 1>,<at 0.7 0.05>]}}})
     )
    ))
   &AddVelocity(Velocity=&MakeVector(
     X=0
     Y=&FloatFromCurve(FloatCurve=DataInterface() {propertyValues {Curve {keys [<at 0 500>,<at 0.7 0>]}}})
     Z=-100
    )) {enabled=false, module=AddVelocity001}
  ]
  particleUpdateScript [
   &ParticleState()
   &InheritSourceMovement(ApplyForce=false, ApplyPosition=true, ApplyPositionFalloff=false)
   &SolveForcesAndVelocity(Clamp Velocity=false, Limit Acceleration=false)
   &ScaleColor(
    Scale Alpha=&FloatFromCurve(
     FloatCurve=DataInterface() {propertyValues {Curve {keys [
         {arriveTangent=-0.0042, interpMode=RCIM_Cubic, leaveTangent=-0.0042, tangentMode=RCTM_User, time=0, value=0}
         {arriveTangent=-0.0104, interpMode=RCIM_Cubic, leaveTangent=-0.0104, tangentMode=RCTM_User, time=0.3, value=1}
         {arriveTangent=0.0525, interpMode=RCIM_Cubic, leaveTangent=0.0525, tangentMode=RCTM_User, time=1, value=0}
        ]}}}
     Scale Curve=&FN_PickFloatDependOnPlatform(TargetPlatformActive=1, TargetPlatformNoActive=0.6)
    )
    Scale RGB=&LerpMultipleVectors(
     Alpha:NiagaraFloat=&Divide_Float(A=$User.VariantSwitch, B=2)
     V 1:Vector3f=&MakeVectorFromLinearColorRGB(Color=&MultiplyLinearColor(
       Linear Color=&ColorFromCurve(ColorCurve=DataInterface() {propertyValues {
          AlphaCurve {keys []}
          BlueCurve {keys [<at 0.0663 0.1 RCIM_Cubic>,<at 0.4163 0.01 RCIM_Cubic>]}
          GreenCurve {keys [<at 0.0663 0.1 RCIM_Cubic>,<at 0.4163 0.01 RCIM_Cubic>]}
          RedCurve {keys [<at 0.0663 0.1 RCIM_Cubic>,<at 0.4163 0.01 RCIM_Cubic>]}
         }})
       Scale Factor=<1 1 1 1>
      ))
     V 2:Vector3f=&MakeVectorFromLinearColorRGB(Color=&ColorFromCurve(ColorCurve=DataInterface() {propertyValues {
         AlphaCurve {keys []}
         BlueCurve {keys [<at 0.3113 0.0724>,<at 0.9816 0.01 RCIM_Cubic>]}
         GreenCurve {keys [<at 0.3113 0.6126>,<at 0.9816 0.01 RCIM_Cubic>]}
         RedCurve {keys [<at 0.3113 1>,<at 0.9816 0.01 RCIM_Cubic>]}
        }}))
     V 3:Vector3f=?
     V 4:Vector3f=?
     V 5:Vector3f=?
     Vector Count:ENiagara_MultipleLerpCount=::NewEnumerator3
    )
    ScaleA=true
    ScaleRGB=true
   )
   &ScaleRibbonWidth(Ribbon Width Scale=&FloatFromCurve(
     FloatCurve=DataInterface() {propertyValues {Curve {keys [
         {arriveTangent=-1.0264, interpMode=RCIM_Cubic, leaveTangent=-1.0264, tangentMode=RCTM_User, time=0, value=0.5}
         {arriveTangent=-0.35, interpMode=RCIM_Cubic, leaveTangent=-0.35, time=0.1, value=1}
         {arriveTangent=-0.3983, interpMode=RCIM_Cubic, leaveTangent=-0.3983, tangentMode=RCTM_User, time=1, value=0.15}
        ]}}}
     Scale Curve=1.8
    ))
  ]
  rendererClasses [&NiagaraRibbonRendererProperties]
  renderers [{rendererClass=&NiagaraRibbonRendererProperties}]
 }
}
summary {
 systemName=NS_Glider_MistMylar_Deploy
 emitters [
  {emitterName=Absorb_Ring, rendererClasses [&NiagaraMeshRendererProperties]}
  {emitterName=Absorb_Smoke, rendererClasses [&NiagaraSpriteRendererProperties]}
  {emitterName=Absorb_Beam, rendererClasses [&NiagaraMeshRendererProperties]}
  {emitterName=Ribbon_Trail, rendererClasses [&NiagaraRibbonRendererProperties]}
  {emitterName=Ribbon_Smoke, rendererClasses [&NiagaraSpriteRendererProperties]}
  {emitterName=Burst_Trail, rendererClasses [&NiagaraRibbonRendererProperties]}
  {emitterName=Burst_Beam, rendererClasses [&NiagaraMeshRendererProperties]}
  {emitterName=Burat_smoke, rendererClasses [&NiagaraSpriteRendererProperties]}
  {emitterName=Ribbon_Smoke001, rendererClasses [&NiagaraSpriteRendererProperties]}
 ]
 userVariables [
  User.bIsFrontEnd {type=NiagaraBool}
  User.bIsFrontEndPreview {type=NiagaraBool}
  User.RibbonIDOffset {type=NiagaraInt32}
  User.VariantSwitch {type=NiagaraFloat, defaultValue=0, description=""}
 ]
}
