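# Niagara system for the MistMylar glider's idle effect, building a dark misty body with smoke and trails from mesh particles attached to the skeletal mesh.
# Burst-spawned center, mask, trail and edge layers persist, while Smoke spawns continuously and drifts along animated offsets.
# Trails write offset positions and their moving average into particle material parameters for lagged motion.
# VariantSwitch brings in gold coloring and extra trails; front-end preview and time-of-day inputs adjust layout and tint.
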NiagaraSystem /BRCosmetics/Gadgets/Assets/VinderTech_GliderChute/Glider_MistMylar/FX/NS_Glider_MistMylar_Idle
dependencies {
 usedDataInterfaces [
  &NiagaraDataInterfaceColorCurve
  &NiagaraDataInterfaceCurve
  &NiagaraDataInterfaceArrayFloat3
  &NiagaraDataInterfacePlatformSet
  &NiagaraDataInterfaceArrayColor
  &NiagaraDataInterfaceArrayFloat
  &NiagaraDataInterfaceSkeletalMesh
 ]
 usedDynamicInputs []
 usedModules [
  &self:SystemSpawnScript.SystemScriptSource.SystemScriptGraph.NiagaraNodeAssignment_1.SetVariables_#1
  &self:SystemSpawnScript.SystemScriptSource.SystemScriptGraph.NiagaraNodeAssignment_2.SetVariables_#2
  &self:SystemSpawnScript.SystemScriptSource.SystemScriptGraph.NiagaraNodeAssignment_0.SetVariables_#3
  &SystemState
 ]
}
emitters {
 Center {
  emitterUpdateScript [
   &EmitterState(
    Inactive Response=::NewEnumerator1
    Life Cycle Mode=::NewEnumerator1
    Loop Behavior=::NewEnumerator1
    Loop Delay=0.7
    Loop Duration Mode=::NewEnumerator1
    UseLoopDelay=true
   )
   &SpawnBurst_Instantaneous(Use Spawn Probability=false, UseLoopCountLimit=false)
  ]
  particleSpawnScript [&InitializeParticle(
    Color Mode=::NewEnumerator0
    Lifetime=5
    Mesh Scale=<1.8 1.8 1.6>
    Mesh Scale Mode=::NewEnumerator4
    Position Offset=<0 55 -50>
    UsePositionOffset=true
    Write Material Random=true
   )]
  particleUpdateScript [
   &ParticleState(Kill Particles When Lifetime Has Elapsed=false, Let Infinitely Lived Particles Die When Emitter Deactivates=true)
   &SkeletalMeshLocation(Bone Sampling Mode=::NewEnumerator4, MaskSpawnGroup=false, Sampled Position Offset=<0 -30 -25>, Skeletal Mesh=$System.Skeletal Mesh)
   &FN_Sprite_Attributes(
    Color=&ColorFromCurve(ColorCurve=DataInterface() {propertyValues {
       AlphaCurve {keys [<at 0 0>,<at 0.05 2 RCIM_Cubic>]}
       BlueCurve {keys [<at 0.2102 0.01 RCIM_Cubic>]}
       GreenCurve {keys [<at 0.2102 0.01 RCIM_Cubic>]}
       RedCurve {keys [<at 0.2102 0.01 RCIM_Cubic>]}
      }})
    Create Color Attribute=true
    Create SpriteSize Attribute=true
    SpriteSize=&Vector2DFromFloat(Value=200)
    Write to Module Namespace=true
   )
   &ScaleMeshSize(Scale Factor=&FN_FrontEndVector3Modifier(Gameplay=<1 1 1>, Front End=<1 1 0.8>))
   &ScaleColor(
    Scale Alpha=&MakeCustomFloatFromBool(
     Bool=$User.bIsFrontEndPreview
     True Float=&FloatFromCurve(FloatCurve=DataInterface() {propertyValues {Curve {keys [<at 0 0>,<at 0.25 1>]}}}, Scale Curve=2)
     False Float=&FloatFromCurve(FloatCurve=DataInterface() {propertyValues {Curve {keys [<at 0 0>,<at 0.05 1>]}}}, Scale Curve=2)
    )
    Scale RGB=&MakeVectorFromLinearColorRGB(Color=&FNTimeOfDayColor(
      Night Color=<rgba 0.003 0.0047 0.01 1>
      Morning Color=<rgba 0.01 0.01 0.01 1>
      Day Color=<rgba 0.01 0.01 0.01 1>
      Evening Color=<rgba 0.01 0.01 0.01 1>
     ))
    ScaleA=true
    ScaleRGB=true
   )
  ]
  rendererClasses [&NiagaraMeshRendererProperties]
  renderers [{rendererClass=&NiagaraMeshRendererProperties, rendererIndex=0}]
 }
 CenterMask {
  emitterUpdateScript [
   &EmitterState(
    Inactive Response=::NewEnumerator1
    Life Cycle Mode=::NewEnumerator1
    Loop Behavior=::NewEnumerator1
    Loop Delay=0.7
    Loop Duration Mode=::NewEnumerator1
    UseLoopDelay=true
   )
   &SpawnBurst_Instantaneous(Use Spawn Probability=false, UseLoopCountLimit=false)
  ]
  particleSpawnScript [&InitializeParticle(
    Color Mode=::NewEnumerator0
    Lifetime=5
    Mesh Scale=<1.05 1.1 1.1>
    Mesh Scale Mode=::NewEnumerator4
    Position Offset=<0 55 -50>
    UsePositionOffset=true
    Write Material Random=true
   )]
  particleUpdateScript [
   &ParticleState(Kill Particles When Lifetime Has Elapsed=false, Let Infinitely Lived Particles Die When Emitter Deactivates=true)
   &ScaleMeshSize(Scale Factor=&MakeCustomVectorFromBool(Bool=$User.bIsFrontEndPreview, True Vector=<0.7 0.7 0.7>, False Vector=<1 1 1>))
   &SkeletalMeshLocation(
    Bone Sampling Mode=::NewEnumerator4
    MaskSpawnGroup=false
    Sampled Position Offset=&MakeCustomVectorFromBool(Bool=$User.bIsFrontEndPreview, True Vector=<0 50 -35>, False Vector=<0 80 -40>)
    Skeletal Mesh=$System.Skeletal Mesh
   )
   &FN_Sprite_Attributes(
    Color=&ColorFromCurve(ColorCurve=DataInterface() {propertyValues {
       AlphaCurve {keys [<at 0 0>,<at 0.05 2 RCIM_Cubic>]}
       BlueCurve {keys [<at 0.2102 0.01 RCIM_Cubic>]}
       GreenCurve {keys [<at 0.2102 0.01 RCIM_Cubic>]}
       RedCurve {keys [<at 0.2102 0.01 RCIM_Cubic>]}
      }})
    Create Color Attribute=true
    Create SpriteSize Attribute=true
    SpriteSize=&Vector2DFromFloat(Value=200)
    Write to Module Namespace=true
   )
   /SetVariables(Particles.Dyn002:Vector4f=&MakeVector4(X=3, Y=10, Z=0, W=0)) {isSetParameters=true}
   /SetVariables(Particles.Color002=&MakeLinearColorFromVectorAndFloat(
     "Vector (RGB)"=&MakeVectorFromLinearColorRGB(Color=&FNTimeOfDayColor(
       Night Color=<rgba 0.0065 0.0074 0.01 8>
       Morning Color=<rgba 0.015 0.015 0.015 8>
       Day Color=<rgba 0.015 0.015 0.015 8>
       Evening Color=<rgba 0.015 0.015 0.015 8>
      ))
     "Float (Alpha)"=&FloatFromCurve(FloatCurve=DataInterface() {propertyValues {Curve {keys [<at 0 0>,<at 0.05 1>]}}}, Scale Curve=5)
    )) {isSetParameters=true}
   &ScaleColor(
    Scale Alpha=&FloatFromCurve(FloatCurve=DataInterface() {propertyValues {Curve {keys [<at 0 0>,<at 0.05 1>]}}})
    Scale RGB=&MakeVectorFromLinearColorRGB(Color=&FNTimeOfDayColor(
      Night Color=<rgba 0.01 0.01 0.01 1>
      Morning Color=<rgba 0.01 0.01 0.01 1>
      Day Color=<rgba 0.01 0.01 0.01 1>
      Evening Color=<rgba 0.01 0.01 0.01 1>
     ))
    ScaleA=true
    ScaleRGB=true
   )
  ]
  rendererClasses [&NiagaraMeshRendererProperties]
  renderers [{rendererClass=&NiagaraMeshRendererProperties, rendererIndex=0}]
 }
 Smoke {
  emitterUpdateScript [
   &EmitterState(
    Delay First Loop Only=true
    Inactive Response=::NewEnumerator1
    Life Cycle Mode=::NewEnumerator1
    Loop Delay=0.5
    Loop Duration=1
    UseLoopDelay=true
   )
   &SpawnRate(SpawnRate=5, Use Spawn Probability=false)
  ]
  particleSpawnScript [
   &PartitionParticles()
   &InitializeParticle(
    Color=&MakeCustomLinearColorFromBool(Bool=$User.bIsFrontEndPreview, True Color=<rgba 0.6129 0.7074 0.8229 1>, False Color=<rgba 0.8229 0.8229 0.8229 1>)
    Lifetime=1.5
    Mesh Scale=<1.5 2 2>
    Mesh Scale Mode=::NewEnumerator4
    UsePositionOffset=true
    Write Material Random=true
   )
   &InitialMeshOrientation(
    Mesh Orientation Mode=::NewEnumerator2
    Rotation=&RandomRangeVector(Minimum=<0 0 0>, Maximum=<1 0 0>)
    Use Rotation Vector=true
   ) {module=InitialMeshOrientation001}
  ]
  particleUpdateScript [
   &ParticleState()
   &ScaleMeshSize(Scale Factor=&VectorFromFloat(Value=&FloatFromCurve(
      FloatCurve=DataInterface() {propertyValues {Curve {keys [<at 0 0.7>,<at 0.2 1.025>,<at 0.4 0.9>,<at 1 0.2>]}}}
      Scale Curve=&RandomRangeFloat(Minimum=0.9, Maximum=1.1)
     )))
   &SkeletalMeshLocation(
    Bone Sampling Mode=::NewEnumerator4
    MaskSpawnGroup=false
    Sampled Position Offset=&MakeVector(
     X=&RandomRangeFloat(Minimum=-15, Maximum=15)
     Y=&FloatFromCurve(
      FloatCurve=DataInterface() {propertyValues {Curve {keys [<at 0 0>,<at 1 270>]}}}
      Scale Curve=&Multiply_Float(
       A=&FN_FrontEndFloatModifier(Gameplay=1.5, Front End=1)
       B=&FN_PickFloatDependOnPlatform(TargetPlatformActive=1, TargetPlatformNoActive=0.9)
      )
     )
     Z=&FloatFromCurve(FloatCurve=DataInterface() {propertyValues {Curve {keys [
          {arriveTangent=4.1302, interpMode=RCIM_Cubic, leaveTangent=4.1302, tangentMode=RCTM_User, time=0, value=-26}
          {arriveTangent=40.7573, interpMode=RCIM_Cubic, leaveTangent=40.7573, tangentMode=RCTM_User, time=1, value=0}
         ]}}})
    )
    Skeletal Mesh=$System.Skeletal Mesh
   )
   &CameraOffset(Camera Offset Amount=&FN_FrontEndFloatModifier(
     Gameplay=&FN_PickFloatDependOnPlatform(TargetPlatformActive=-5, TargetPlatformNoActive=-10)
     Front End=-50
    ))
   &ScaleColor(
    Scale RGB=/Niagara/DynamicInputs/Math/LerpMultipleVectors(
     Alpha:NiagaraFloat=&Divide_Float(A=$User.VariantSwitch, B=2)
     V 1:Vector3f=&MakeVectorFromLinearColorRGB(Color=&FNTimeOfDayColor(Night Color=<rgba 0.15 0.2375 0.5 1>, Morning Color=<rgba 1 1 1 1>, Day Color=<rgba 1 1 1 1>, Evening Color=<rgba 1 1 1 1>))
     V 2:Vector3f=&MakeVectorFromLinearColorRGB(Color=&MultiplyLinearColor(
       Linear Color=&ColorFromCurve(ColorCurve=DataInterface() {propertyValues {
          AlphaCurve {keys []}
          BlueCurve {keys [{arriveTangent=-0.2522, time=0.3558, value=1}, {arriveTangent=-4.6751, time=0.5379, value=0.1478}]}
          GreenCurve {keys [{arriveTangent=-3.3642, time=0.3558, value=1}, {arriveTangent=-14.5607, time=0.5379, value=1.6914}]}
          RedCurve {keys [{arriveTangent=-6.5976, time=0.3558, value=1}, {arriveTangent=-27.0666, time=0.5379, value=3}]}
         }})
       Scale Factor=&MakeVector4_FromLinearColor(LinearColor=&FNTimeOfDayColor(Night Color=<rgba 0.2 0.275 0.5 1>, Morning Color=<rgba 1 1 1 1>, Day Color=<rgba 1 1 1 1>, Evening Color=<rgba 1 1 1 1>))
      ))
     V 3:Vector3f=?
     V 4:Vector3f=?
     V 5:Vector3f=?
     Vector Count:ENiagara_MultipleLerpCount=::NewEnumerator3
    )
    ScaleA=true
    ScaleRGB=true
   ) {module=ScaleColor001}
   &ScaleColor(
    Scale Alpha=&FloatFromCurve(FloatCurve=DataInterface() {propertyValues {Curve {keys [
         {arriveTangent=5.611, interpMode=RCIM_Cubic, leaveTangent=5.611, tangentMode=RCTM_User, time=0, value=0}
         {arriveTangent=0.0117, interpMode=RCIM_Cubic, leaveTangent=0.0117, tangentMode=RCTM_User, time=0.3, value=1}
         <at 1 0 RCIM_Cubic>
        ]}}})
    Scale RGB=&VectorFromFloat(Value=&Multiply_Float(A=&RandomRangeFloat(Minimum=0.8, Maximum=1), B=1))
    ScaleA=true
    ScaleRGB=true
   )
   &DynamicMaterialParameters(
    Index 0 Param 1=&FloatFromCurve(
     FloatCurve=DataInterface() {propertyValues {Curve {keys [
         {interpMode=RCIM_Cubic, tangentMode=RCTM_User, time=0, value=0}
         <at 0.2 0.55>
         {arriveTangent=-1.0223, interpMode=RCIM_Cubic, leaveTangent=-1.0223, tangentMode=RCTM_User, time=1, value=1}
        ]}}}
     Scale Curve=&RandomRangeFloat(Minimum=2, Maximum=5)
    )
    Index 0 Param 3=1
    Index 0 Param 4=&FloatFromCurve(FloatCurve=DataInterface() {propertyValues {Curve {keys [{arriveTangent=-2.87, interpMode=RCIM_Cubic, leaveTangent=-2.87, tangentMode=RCTM_User, time=0.3, value=1}, <at 1 0 RCIM_Cubic>]}}})
    Module.Index0Param0Write=true
    Module.Index0Param1Write=true
    Module.Index0Param2Write=true
    Module.Index0Param3Write=true
   )
  ]
  rendererClasses [&NiagaraMeshRendererProperties]
  renderers [{rendererClass=&NiagaraMeshRendererProperties, rendererIndex=0}]
 }
 Trail {
  emitterUpdateScript [
   &EmitterState(
    Inactive Response=::NewEnumerator1
    Life Cycle Mode=::NewEnumerator1
    Loop Behavior=::NewEnumerator1
    Loop Delay=0.6
    Loop Duration Mode=::NewEnumerator1
    UseLoopDelay=true
   )
   &SpawnBurst_Instantaneous(Spawn Count=8, Use Spawn Probability=false, UseLoopCountLimit=false)
   /SetVariables(Emitter.PositionFix:NiagaraFloat=&FN_FrontEndFloatModifier(Gameplay=1, Front End=1)) {isSetParameters=true, module=SetVariables_#4}
  ]
  particleSpawnScript [
   &PartitionParticles(Number Of Partitions=8)
   &InitializeParticle(
    Color Maximum=<rgba 0.015 0.015 0.015 2>
    Color Minimum=<rgba 0.01 0.01 0.01 1.5>
    Color Mode=::NewEnumerator2
    Mesh Scale=&SelectVectorFromArray(
     Direct Array Index=$Particles.PartitionParticles.Partition
     Vector Selection Array=DataInterface() {propertyValues {FloatData [<2 2.2 2>, <1 1.3 1.5>, <1 1.3 1>, <1 1.5 1>, <1 1.8 1>, <1 1.5 1>, <1 1.5 1>, <1 1.2 1>]}}
    )
    Mesh Scale Mode=::NewEnumerator4
    UsePositionOffset=true
    Write Material Random=true
   )
   &InitialMeshOrientation(
    Mesh Orientation Mode=::NewEnumerator2
    Rotation=&MakeCustomVectorFromBool(
     Bool=$User.bIsFrontEndPreview
     True Vector=&SelectVectorFromArray(
      Direct Array Index=$Particles.PartitionParticles.Partition
      Vector Selection Array=DataInterface() {propertyValues {FloatData [<0 0 0>, <0 -0.2 0.02>, <0 0.4 -0.02>, <0 0 -0.08>, <-0.05 0.3 0.05>, <0 0.3 0>, <0 0 0.02>, <0.06 0.5 0.036>]}}
     )
     False Vector=&SelectVectorFromArray(
      Direct Array Index=$Particles.PartitionParticles.Partition
      Vector Selection Array=DataInterface() {propertyValues {FloatData [<0 0 0>, <0 -0.2 0.02>, <0 0.4 -0.02>, <0 0 -0.02>, <0.01 0.2 0>, <0 0.3 0>, <0 0 0.02>, <0.06 0.5 -0.03>]}}
     )
    )
    Use Rotation Vector=true
   )
  ]
  particleUpdateScript [
   &ParticleState(Kill Particles When Lifetime Has Elapsed=false, Let Infinitely Lived Particles Die When Emitter Deactivates=true)
   &SkeletalMeshLocation(
    Bone Sampling Mode=::NewEnumerator4
    MaskSpawnGroup=false
    Sampled Position Offset=&SelectVectorFromArray(
     Direct Array Index=$Particles.PartitionParticles.Partition
     Vector Selection Array=DataInterface() {propertyValues {FloatData [<0 -20 -15>, <-30 0 0>, <50 10 10>, <15 -60 -37>, <-30 -40 -35>, <30 0 -50>, <-55 -1.6875 -15>, <54.5358 9.9689 -15>]}}
    )
    Skeletal Mesh=$System.Skeletal Mesh
   )
   &DynamicMaterialParameters(
    Index 0 Param 1=&RandomRangeFloat(Minimum=0.2, Maximum=1)
    Index 0 Param 2=&RandomRangeFloat(Minimum=0.2, Maximum=1)
    Index 0 Param 3=1
    Index 0 Param 4=20
    Index 2 Param 1=&MakeFloatFromVector(
     Vector=&AddVectorToPosition(
      Position=$Particles.Position
      Vector=&TransformVector(
       Vector=&FN_FrontEndVector3Modifier(Gameplay=$System.IGPOFix, Front End=$System.FEPoFix)
       Destination Space=::Simulation
      )
     )
     Channel=::NewEnumerator0
    )
    Index 2 Param 2=&MakeFloatFromVector(
     Vector=&AddVectorToPosition(
      Position=$Particles.Position
      Vector=&TransformVector(
       Vector=&FN_FrontEndVector3Modifier(Gameplay=$System.IGPOFix, Front End=$System.FEPoFix)
       Destination Space=::Simulation
      )
     )
     Channel=::NewEnumerator1
    )
    Index 2 Param 3=&MakeFloatFromVector(
     Vector=&AddVectorToPosition(
      Position=$Particles.Position
      Vector=&TransformVector(
       Vector=&FN_FrontEndVector3Modifier(Gameplay=$System.IGPOFix, Front End=$System.FEPoFix)
       Destination Space=::Simulation
      )
     )
     Channel=::NewEnumerator2
    )
    Index 3 Param 1=&MakeCustomFloatFromBool(
     Bool=$User.bIsFrontEndPreview
     True Float=&FloatFromCurve(
      FloatCurve=DataInterface() {propertyValues {Curve {keys [
          {arriveTangent=237.3113, interpMode=RCIM_Cubic, leaveTangent=237.3113, tangentMode=RCTM_User, time=0, value=-800}
          {arriveTangent=816.2405, interpMode=RCIM_Cubic, leaveTangent=816.2405, tangentMode=RCTM_User, time=1, value=0}
         ]}}}
      Scale Curve=&RandomRangeFloat(Minimum=0.8, Maximum=1)
     )
     False Float=&FloatFromCurve(
      FloatCurve=DataInterface() {propertyValues {Curve {keys [
          {arriveTangent=237.3113, interpMode=RCIM_Cubic, leaveTangent=237.3113, tangentMode=RCTM_User, time=0, value=-600}
          {arriveTangent=816.2405, interpMode=RCIM_Cubic, leaveTangent=816.2405, tangentMode=RCTM_User, time=1, value=0}
         ]}}}
      Scale Curve=&RandomRangeFloat(Minimum=0.8, Maximum=1)
     )
    )
    Module.Index0Param0Write=true
    Module.Index0Param1Write=true
    Module.Index0Param2Write=true
    Module.Index0Param3Write=true
    Module.Index2Param0Write=true
    Module.Index2Param1Write=true
    Module.Index2Param2Write=true
    Module.Index2Param3Write=true
    Module.Index3Param0Write=true
    Module.Index3Param1Write=true
    Module.Index3Param2Write=true
    Module.Index3Param3Write=true
    Write Parameter Index 2=true
    Write Parameter Index 3=true
   )
   &InterpolateOverTime(
    Rate Of Change:NiagaraFloat=&FN_FrontEndFloatModifier(Gameplay=10, Front End=20)
    Target Vector Value:Vector3f=&AddVectorToPosition(
     Position=$Particles.Position
     Vector=&TransformVector(
      Vector=&FN_FrontEndVector3Modifier(Gameplay=$System.IGPOFix, Front End=$System.FEPoFix)
      Destination Space=::Simulation
     )
    )
    Variable Type=::NewEnumerator2
   )
   /SetVariables(Particles.Interpolate Dynamic=&MakeFloat4FromVectorAndFloat(
     "Vector (RGB)"=$StackContext.InterpolateOverTime.InterpolateOverTime_V3_.Moving Average
     "Float (Alpha)"=&FN_FrontEndFloatModifier(Gameplay=3, Front End=5)
    )) {isSetParameters=true, module=SetVariables_#5}
   /SetVariables(Particles.Color002=&MultiplyLinearColor(
     Linear Color=&MakeCustomLinearColorFromBool(
      Bool=$User.bIsFrontEndPreview
      True Color=<rgba 0.014 0.0167 0.02 0.6>
      False Color=&FNTimeOfDayColor(
       Night Color=<rgba 0.009 0.0142 0.03 0.6>
       Morning Color=<rgba 0.03 0.03 0.03 0.6>
       Day Color=<rgba 0.03 0.03 0.03 0.6>
       Evening Color=<rgba 0.03 0.03 0.03 0.6>
      )
     )
     Scale Factor=&MakeVector4_FromFloat(Float=&FNTimeOfDayScalar(Night Scale=0.6, Morning Scale=1, Evening Scale=1))
    )) {isSetParameters=true}
   &ScaleColor(
    Scale Alpha=&FloatFromCurve(
     FloatCurve=DataInterface() {propertyValues {Curve {keys [<at 0 0>,<at 0.1 1>]}}}
     Scale Curve=&FN_FrontEndFloatModifier(Gameplay=1, Front End=0.5)
    )
    Scale RGB=&MakeVectorFromLinearColorRGB(Color=&MultiplyLinearColor(
      Linear Color=&MakeCustomLinearColorFromBool(
       Bool=$User.bIsFrontEndPreview
       True Color=<rgba 0.6023 0.7578 0.9479 1>
       False Color=&FNTimeOfDayColor(Night Color=<rgba 0.3 0.475 1 1>, Morning Color=<rgba 1 1 1 1>, Day Color=<rgba 1 1 1 1>, Evening Color=<rgba 1 1 1 1>)
      )
      Scale Factor=&MakeVector4_FromFloat(Float=&FNTimeOfDayScalar(Night Scale=0.6, Morning Scale=1, Evening Scale=1))
     ))
    ScaleA=true
    ScaleRGB=true
   )
   &ScaleMeshSize(Scale Factor=&FN_FrontEndVector3Modifier(Gameplay=<1 1 1>, Front End=<1 0.6 1>))
  ]
  rendererClasses [&NiagaraMeshRendererProperties]
  renderers [
   {rendererClass=&NiagaraMeshRendererProperties, rendererIndex=0}
   {rendererClass=&NiagaraMeshRendererProperties, rendererIndex=1}
  ]
 }
 Trail002 {
  emitterUpdateScript [
   &EmitterState(
    Inactive Response=::NewEnumerator1
    Life Cycle Mode=::NewEnumerator1
    Loop Behavior=::NewEnumerator1
    Loop Delay=0.6
    Loop Duration Mode=::NewEnumerator1
    UseLoopDelay=true
   )
   &SpawnBurst_Instantaneous(
    Spawn Count=8
    Spawn Probability=&Lerp_Float(A=0, B=1, Alpha=$User.VariantSwitch)
    Use Spawn Probability=true
    UseLoopCountLimit=false
   )
   /SetVariables(Emitter.PositionFix:NiagaraFloat=&FN_FrontEndFloatModifier(Gameplay=1, Front End=1)) {isSetParameters=true, module=SetVariables_#4}
  ]
  particleSpawnScript [
   &PartitionParticles(Number Of Partitions=8)
   &InitializeParticle(
    Color Maximum=<rgba 0.015 0.015 0.015 2>
    Color Minimum=<rgba 0.01 0.01 0.01 1.5>
    Color Mode=::NewEnumerator2
    Mesh Scale=&SelectVectorFromArray(
     Direct Array Index=$Particles.PartitionParticles.Partition
     Vector Selection Array=DataInterface() {propertyValues {FloatData [<2 2.2 2>, <1 1.3 1.5>, <1 1.3 1>, <1 1.3 1>, <1 1.5 1>, <1 1.5 1>, <1 1.5 1>, <1 1.2 1>]}}
    )
    Mesh Scale Mode=::NewEnumerator4
    UsePositionOffset=true
    Write Material Random=true
   )
   &InitialMeshOrientation(
    Mesh Orientation Mode=::NewEnumerator2
    Rotation=&SelectVectorFromArray(
     Direct Array Index=$Particles.PartitionParticles.Partition
     Vector Selection Array=DataInterface() {propertyValues {FloatData [<0 0.25 0>, <0 0.3 0>, <0 0 0>, <0 0 0>, <0 0 0>, <0 0 0>, <0 0 0>, <0 0 0>]}}
    )
    Use Rotation Vector=true
   ) {module=InitialMeshOrientation001}
   &InitialMeshOrientation(
    Mesh Orientation Mode=::NewEnumerator2
    Rotation=&MakeCustomVectorFromBool(
     Bool=$User.bIsFrontEndPreview
     True Vector=&SelectVectorFromArray(
      Direct Array Index=$Particles.PartitionParticles.Partition
      Vector Selection Array=DataInterface() {propertyValues {FloatData [<0 0 0>, <0 -0.2 0.02>, <0 0.4 -0.02>, <0 0 -0.08>, <-0.05 0.3 0.05>, <0 0.3 0>, <0 0 0.02>, <0.06 0.5 0.036>]}}
     )
     False Vector=&SelectVectorFromArray(
      Direct Array Index=$Particles.PartitionParticles.Partition
      Vector Selection Array=DataInterface() {propertyValues {FloatData [<0 0 0>, <0 -0.2 0.02>, <0 0.4 -0.02>, <0 0 -0.02>, <0.01 0.2 0>, <0 0.3 0>, <0 0 0.02>, <0.06 0.5 -0.03>]}}
     )
    )
    Use Rotation Vector=true
   )
   &CameraOffset(Camera Offset Amount=5)
  ]
  particleUpdateScript [
   &ParticleState(Kill Particles When Lifetime Has Elapsed=false, Let Infinitely Lived Particles Die When Emitter Deactivates=true)
   &SkeletalMeshLocation(
    Bone Sampling Mode=::NewEnumerator4
    MaskSpawnGroup=false
    Sampled Position Offset=&SelectVectorFromArray(
     Direct Array Index=$Particles.PartitionParticles.Partition
     Vector Selection Array=DataInterface() {propertyValues {FloatData [<0 -20 -15>, <-30 0 0>, <50 10 10>, <15 -60 -37>, <-30 -40 -35>, <30 0 -50>, <-55 -1.6875 -15>, <54.5358 9.9689 -15>]}}
    )
    Skeletal Mesh=$System.Skeletal Mesh
   )
   &DynamicMaterialParameters(
    Index 0 Param 1=&RandomRangeFloat(Minimum=0.2, Maximum=1)
    Index 0 Param 2=&RandomRangeFloat(Minimum=0.2, Maximum=1)
    Index 0 Param 3=1
    Index 0 Param 4=20
    Index 2 Param 1=&MakeFloatFromVector(
     Vector=&AddVectorToPosition(
      Position=$Particles.Position
      Vector=&TransformVector(
       Vector=&FN_FrontEndVector3Modifier(Gameplay=$System.IGPOFix, Front End=$System.FEPoFix)
       Destination Space=::Simulation
      )
     )
     Channel=::NewEnumerator0
    )
    Index 2 Param 2=&MakeFloatFromVector(
     Vector=&AddVectorToPosition(
      Position=$Particles.Position
      Vector=&TransformVector(
       Vector=&FN_FrontEndVector3Modifier(Gameplay=$System.IGPOFix, Front End=$System.FEPoFix)
       Destination Space=::Simulation
      )
     )
     Channel=::NewEnumerator1
    )
    Index 2 Param 3=&MakeFloatFromVector(
     Vector=&AddVectorToPosition(
      Position=$Particles.Position
      Vector=&TransformVector(
       Vector=&FN_FrontEndVector3Modifier(Gameplay=$System.IGPOFix, Front End=$System.FEPoFix)
       Destination Space=::Simulation
      )
     )
     Channel=::NewEnumerator2
    )
    Index 3 Param 1=&MakeCustomFloatFromBool(
     Bool=$User.bIsFrontEndPreview
     True Float=&FloatFromCurve(
      FloatCurve=DataInterface() {propertyValues {Curve {keys [
          {arriveTangent=237.3113, interpMode=RCIM_Cubic, leaveTangent=237.3113, tangentMode=RCTM_User, time=0, value=-800}
          {arriveTangent=816.2405, interpMode=RCIM_Cubic, leaveTangent=816.2405, tangentMode=RCTM_User, time=1, value=0}
         ]}}}
      Scale Curve=&RandomRangeFloat(Minimum=0.8, Maximum=1)
     )
     False Float=&FloatFromCurve(
      FloatCurve=DataInterface() {propertyValues {Curve {keys [
          {arriveTangent=237.3113, interpMode=RCIM_Cubic, leaveTangent=237.3113, tangentMode=RCTM_User, time=0, value=-600}
          {arriveTangent=816.2405, interpMode=RCIM_Cubic, leaveTangent=816.2405, tangentMode=RCTM_User, time=1, value=0}
         ]}}}
      Scale Curve=&RandomRangeFloat(Minimum=0.8, Maximum=1)
     )
    )
    Module.Index0Param0Write=true
    Module.Index0Param1Write=true
    Module.Index0Param2Write=true
    Module.Index0Param3Write=true
    Module.Index2Param0Write=true
    Module.Index2Param1Write=true
    Module.Index2Param2Write=true
    Module.Index2Param3Write=true
    Module.Index3Param0Write=true
    Module.Index3Param1Write=true
    Module.Index3Param2Write=true
    Module.Index3Param3Write=true
    Write Parameter Index 2=true
    Write Parameter Index 3=true
   )
   &InterpolateOverTime(
    Rate Of Change:NiagaraFloat=&FN_FrontEndFloatModifier(Gameplay=10, Front End=10)
    Target Vector Value:Vector3f=&AddVectorToPosition(
     Position=$Particles.Position
     Vector=&TransformVector(
      Vector=&FN_FrontEndVector3Modifier(Gameplay=$System.IGPOFix, Front End=$System.FEPoFix)
      Destination Space=::Simulation
     )
    )
    Variable Type=::NewEnumerator2
   )
   /SetVariables(Particles.Interpolate Dynamic=&MakeFloat4FromVectorAndFloat(
     "Vector (RGB)"=$StackContext.InterpolateOverTime.InterpolateOverTime_V3_.Moving Average
     "Float (Alpha)"=&FN_FrontEndFloatModifier(Gameplay=3, Front End=1)
    )) {isSetParameters=true, module=SetVariables_#5}
   &ScaleMeshSize(Scale Factor=&FN_FrontEndVector3Modifier(
     Gameplay=<1 1 1>
     Front End=&SelectVectorFromArray(
      Direct Array Index=$Particles.PartitionParticles.Partition
      Vector Selection Array=DataInterface() {propertyValues {FloatData [<0.6 0.6 0.6>, <0.6 0.6 0.6>, <0.6 0.7 0.6>, <0.4 0.5 0.4>, <0.6 0.5 0.6>, <0.6 0.5 0.6>, <0.6 0.5 0.6>, <0.5 0.4 0.5>]}}
     )
    ))
   /SetVariables(Particles.Color_Golden:LinearColor=&MakeCustomLinearColorFromBool(
     Bool=$User.bIsFrontEndPreview
     True Color=&RandomRangeLinearColor(Minimum=<rgba 1 0.35 0 0.5>, Maximum=<rgba 2 0.75 0 1>)
     False Color=&RandomRangeLinearColor(Minimum=<rgba 1 0.4 0.0208 0.5>, Maximum=<rgba 1 0.5307 0.062 1>)
    )) {isSetParameters=true}
   /SetVariables(Particles.DetailColor_Golden:LinearColor=&MakeCustomLinearColorFromBool(
     Bool=$User.bIsFrontEndPreview
     True Color=&RandomRangeLinearColor(Minimum=<rgba 1 0.3 0.0306 0.8>, Maximum=<rgba 2 1.1 0 1>)
     False Color=&RandomRangeLinearColor(Minimum=<rgba 1 0.5883 0.0611 0.3>, Maximum=<rgba 3 1.7 0.1223 0.6>)
    )) {isSetParameters=true}
  ]
  rendererClasses [&NiagaraMeshRendererProperties]
  renderers [
   {rendererClass=&NiagaraMeshRendererProperties, rendererIndex=0}
   {rendererClass=&NiagaraMeshRendererProperties, rendererIndex=1}
  ]
 }
 Trail_Dark {
  emitterUpdateScript [
   &EmitterState(
    Inactive Response=::NewEnumerator1
    Life Cycle Mode=::NewEnumerator1
    Loop Behavior=::NewEnumerator1
    Loop Delay=0.7
    Loop Duration Mode=::NewEnumerator1
    UseLoopDelay=true
   )
   &SpawnBurst_Instantaneous(Spawn Count=3, Use Spawn Probability=false, UseLoopCountLimit=false)
   /SetVariables(Emitter.PositionFix:NiagaraFloat=&FN_FrontEndFloatModifier(Gameplay=1, Front End=1)) {isSetParameters=true}
  ]
  particleSpawnScript [
   &PartitionParticles(Number Of Partitions=3)
   &InitializeParticle(
    Color=&FNTimeOfDayColor(
     Night Color=<rgba 0.003 0.0047 0.01 4>
     Morning Color=<rgba 0.01 0.01 0.01 4>
     Day Color=<rgba 0.01 0.01 0.01 4>
     Evening Color=<rgba 0.01 0.01 0.01 4>
    )
    Mesh Scale=&SelectVectorFromArray(
     Direct Array Index=$Particles.PartitionParticles.Partition
     Vector Selection Array=DataInterface() {propertyValues {FloatData [<1 1 8>, <1 1 9>, <1 1 10>]}}
    )
    Mesh Scale Mode=::NewEnumerator4
    UsePositionOffset=true
    Write Material Random=true
   )
   &InitialMeshOrientation(Mesh Orientation Mode=::NewEnumerator2, Rotation=<-0.25 0 0>, Use Rotation Vector=true)
   &CameraOffset(Camera Offset Amount=&MakeCustomFloatFromBool(Bool=$User.bIsFrontEndPreview, True Float=-20, False Float=0))
  ]
  particleUpdateScript [
   &ParticleState(Kill Particles When Lifetime Has Elapsed=false, Let Infinitely Lived Particles Die When Emitter Deactivates=true)
   &SkeletalMeshLocation(
    Bone Sampling Mode=::NewEnumerator4
    MaskSpawnGroup=false
    Sampled Position Offset=&SelectVectorFromArray(
     Direct Array Index=$Particles.PartitionParticles.Partition
     Vector Selection Array=DataInterface() {propertyValues {FloatData [<-36 0 20>, <40 0 -20>, <0 0 -10>]}}
    )
    Skeletal Mesh=$System.Skeletal Mesh
   )
   &DynamicMaterialParameters(
    Index 0 Param 1=-1.5
    Index 0 Param 2=1
    Index 0 Param 3=1.5
    Index 0 Param 4=1
    Index 2 Param 1=&MakeFloatFromVector(
     Vector=&AddVectorToPosition(
      Position=$Particles.Position
      Vector=&TransformVector(
       Vector=&FN_FrontEndVector3Modifier(Gameplay=$System.IGPOFix, Front End=$System.FEPoFix)
       Destination Space=::Simulation
      )
     )
     Channel=::NewEnumerator0
    )
    Index 2 Param 2=&MakeFloatFromVector(
     Vector=&AddVectorToPosition(
      Position=$Particles.Position
      Vector=&TransformVector(
       Vector=&FN_FrontEndVector3Modifier(Gameplay=$System.IGPOFix, Front End=$System.FEPoFix)
       Destination Space=::Simulation
      )
     )
     Channel=::NewEnumerator1
    )
    Index 2 Param 3=&MakeFloatFromVector(
     Vector=&AddVectorToPosition(
      Position=$Particles.Position
      Vector=&TransformVector(
       Vector=&FN_FrontEndVector3Modifier(Gameplay=$System.IGPOFix, Front End=$System.FEPoFix)
       Destination Space=::Simulation
      )
     )
     Channel=::NewEnumerator2
    )
    Index 3 Param 1=&MakeCustomFloatFromBool(
     Bool=$User.bIsFrontEndPreview
     True Float=&FloatFromCurve(FloatCurve=DataInterface() {propertyValues {Curve {keys [<at 0 -600>,<at 1 0>]}}})
     False Float=&FloatFromCurve(FloatCurve=DataInterface() {propertyValues {Curve {keys [<at 0 -500>,<at 1 0>]}}})
    )
    Module.Index0Param0Write=true
    Module.Index0Param1Write=true
    Module.Index0Param2Write=true
    Module.Index0Param3Write=true
    Module.Index2Param0Write=true
    Module.Index2Param1Write=true
    Module.Index2Param2Write=true
    Module.Index2Param3Write=true
    Module.Index3Param0Write=true
    Module.Index3Param1Write=true
    Module.Index3Param2Write=true
    Module.Index3Param3Write=true
    Write Parameter Index 2=true
    Write Parameter Index 3=true
   )
   &InterpolateOverTime(
    Rate Of Change:NiagaraFloat=&FN_FrontEndFloatModifier(Gameplay=10, Front End=10)
    Target Vector Value:Vector3f=&AddVectorToPosition(
     Position=$Particles.Position
     Vector=&TransformVector(
      Vector=&FN_FrontEndVector3Modifier(Gameplay=$System.IGPOFix, Front End=$System.FEPoFix)
      Destination Space=::Simulation
     )
    )
    Variable Type=::NewEnumerator2
   )
   /SetVariables(Particles.Interpolate Dynamic=&MakeFloat4FromVectorAndFloat(
     "Vector (RGB)"=$StackContext.InterpolateOverTime.InterpolateOverTime_V3_.Moving Average
     "Float (Alpha)"=&FN_FrontEndFloatModifier(Gameplay=0.6, Front End=&Lerp_Float(A=0, B=2, Alpha=$User.VariantSwitch))
    )) {isSetParameters=true}
   /SetVariables(Particles.Color_Golden:LinearColor=&MakeCustomLinearColorFromBool(Bool=$User.bIsFrontEndPreview, True Color=<rgba 2 0.718 0 2>, False Color=<rgba 5 2.1949 0.2456 1>)) {isSetParameters=true}
   /SetVariables(Particles.Size_Golden:Vector3f=&Multiply_Vector(
     A=&SelectVectorFromArray(
      Direct Array Index=&MakeIntFromFloat(Float=$User.VariantSwitch)
      Vector Selection Array=DataInterface() {propertyValues {FloatData [<0 0 0>, <1 1 9>]}}
     )
     B=&MakeCustomVectorFromBool(Bool=$User.bIsFrontEndPreview, True Vector=<1 1 0.8>, False Vector=<1 1 1>)
    )) {isSetParameters=true}
  ]
  rendererClasses [&NiagaraMeshRendererProperties]
  renderers [
   {rendererClass=&NiagaraMeshRendererProperties, rendererIndex=0}
   {rendererClass=&NiagaraMeshRendererProperties, rendererIndex=1}
  ]
 }
 Trail_Edge {
  emitterUpdateScript [
   &EmitterState(
    Inactive Response=::NewEnumerator1
    Life Cycle Mode=::NewEnumerator1
    Loop Behavior=::NewEnumerator1
    Loop Delay=0.1
    Loop Duration Mode=::NewEnumerator1
    UseLoopDelay=true
   )
   &SpawnBurst_Instantaneous(
    Spawn Count=6
    Spawn Probability=&MakeCustomFloatFromBool(Bool=$User.bIsFrontEndPreview, True Float=1, False Float=1)
    Use Spawn Probability=true
    UseLoopCountLimit=false
   )
   /SetVariables(Emitter.Fade:NiagaraFloat=&FN_FrontEndFloatModifier(Gameplay=0, Front End=1)) {isSetParameters=true}
   /SetVariables(Emitter.PositionFix:NiagaraFloat=&FN_FrontEndFloatModifier(Gameplay=1, Front End=1)) {isSetParameters=true}
  ]
  particleSpawnScript [
   &PartitionParticles(Number Of Partitions=6)
   &InitializeParticle(
    Color=&MakeCustomLinearColorFromBool(
     Bool=$User.bIsFrontEndPreview
     True Color=&SelectLinearColorFromArray(
      Array Sampling Mode=::NewEnumerator1
      Direct Array Index=&MakeIntFromFloat(Float=$User.VariantSwitch)
      Color Selection Array=DataInterface() {propertyValues {ColorData [<rgba 0.5156 0.5156 0.5156 1.6>, <rgba 8 3.8197 0.4345 1>]}}
      Random Seed=?
     )
     False Color=&SelectLinearColorFromArray(
      Array Sampling Mode=::NewEnumerator1
      Direct Array Index=&MakeIntFromFloat(Float=$User.VariantSwitch)
      Color Selection Array=DataInterface() {propertyValues {ColorData [<rgba 0.5156 0.5156 0.5156 1.6>, <rgba 8 4 0.3678 1>]}}
      Random Seed=?
     )
    )
    Mesh Scale=&SelectVectorFromArray(
     Direct Array Index=$Particles.PartitionParticles.Partition
     Vector Selection Array=DataInterface() {propertyValues {FloatData [<1 1 0.7>, <1 1 0.7>, <1 1 1.2>, <1 1 1.2>, <1.5 1.5 1.5>, <1.5 1.5 1.5>]}}
    )
    Mesh Scale Mode=::NewEnumerator4
    UsePositionOffset=false
    Write Material Random=true
   )
   &DynamicMaterialParameters(
    Index 0 Param 1=&RandomRangeFloat(Minimum=0, Maximum=1)
    Index 0 Param 2=&RandomRangeFloat(Minimum=0, Maximum=10)
    Index 0 Param 3=30
    Index 0 Param 4=1
    Module.Param0WriteEnabled=true
    Module.Param1WriteEnabled=true
    Module.Param2WriteEnabled=false
    Module.Param3WriteEnabled=true
   ) {enabled=false}
   &InitialMeshOrientation(Mesh Orientation Mode=::NewEnumerator2, Rotation=<-0.25 0 0>, Use Rotation Vector=true)
   &InitialMeshOrientation(
    Mesh Orientation Mode=::NewEnumerator2
    Rotation=&RandomRangeVector(Minimum=<0 0 0>, Maximum=<0 5 0>)
    Use Rotation Vector=true
   ) {module=InitialMeshOrientation002}
   &InitialMeshOrientation(
    Mesh Orientation Mode=::NewEnumerator2
    Rotation=&SelectVectorFromArray(
     Direct Array Index=$Particles.PartitionParticles.Partition
     Vector Selection Array=DataInterface() {propertyValues {FloatData [<0.03 0 0.1>, <0.03 0 -0.1>, <0.03 0 0.07>, <0.03 0 -0.07>, <0 0 0.05>, <0 0 -0.06>]}}
    )
    Use Rotation Vector=true
   ) {module=InitialMeshOrientation001}
  ]
  particleUpdateScript [
   &ParticleState(Kill Particles When Lifetime Has Elapsed=false, Let Infinitely Lived Particles Die When Emitter Deactivates=true)
   &ScaleMeshSize(Scale Factor=&MakeCustomVectorFromBool(
     Bool=$User.bIsFrontEndPreview
     True Vector=&Lerp_Vector(A=<1 1 1>, B=<1 1 0.85>, Alpha=$User.VariantSwitch)
     False Vector=<1 1 1>
    ))
   &SkeletalMeshLocation(
    Bone Sampling Mode=::NewEnumerator4
    MaskSpawnGroup=false
    Sampled Position Offset=/Niagara/DynamicInputs/TypeConversions/MakeCustomPositionFromBool(
     Bool:NiagaraBool=$User.bIsFrontEndPreview
     True Position:NiagaraPosition=&SelectVectorFromArray(
      Direct Array Index=$Particles.PartitionParticles.Partition
      Vector Selection Array=DataInterface() {propertyValues {FloatData [<-15 50 15>, <15 50 15>, <-30 20 -15>, <30 20 -15>, <-30 0 -40>, <30 0 -40>]}}
     )
     False Position:NiagaraPosition=&SelectVectorFromArray(
      Direct Array Index=$Particles.PartitionParticles.Partition
      Vector Selection Array=DataInterface() {propertyValues {FloatData [<-35 50 15>, <35 50 15>, <-50 20 -15>, <50 20 -15>, <-50 0 -40>, <50 0 -40>]}}
     )
    )
    Skeletal Mesh=$System.Skeletal Mesh
   )
   &DynamicMaterialParameters(
    Index 0 Param 1=&RandomRangeFloat(Minimum=0.5, Maximum=1)
    Index 0 Param 2=&RandomRangeFloat(Minimum=0.5, Maximum=1)
    Index 0 Param 3=&Multiply_Float(
     A=&SelectFloatFromArray(
      Direct Array Index=$Particles.PartitionParticles.Partition
      Float Selection Array=DataInterface() {propertyValues {FloatData [5,5,8,8,10,10]}}
     )
     B=&Lerp_Float(A=0.5, B=1, Alpha=1)
    )
    Index 0 Param 4=1
    Index 2 Param 1=&MakeFloatFromVector(
     Vector=&AddVectorToPosition(
      Position=$Particles.Position
      Vector=&TransformVector(
       Vector=&FN_FrontEndVector3Modifier(Gameplay=$System.IGPOFix, Front End=$System.FEPoFix)
       Destination Space=::Simulation
      )
     )
     Channel=::NewEnumerator0
    )
    Index 2 Param 2=&MakeFloatFromVector(
     Vector=&AddVectorToPosition(
      Position=$Particles.Position
      Vector=&TransformVector(
       Vector=&FN_FrontEndVector3Modifier(Gameplay=$System.IGPOFix, Front End=$System.FEPoFix)
       Destination Space=::Simulation
      )
     )
     Channel=::NewEnumerator1
    )
    Index 2 Param 3=&MakeFloatFromVector(
     Vector=&AddVectorToPosition(
      Position=$Particles.Position
      Vector=&TransformVector(
       Vector=&FN_FrontEndVector3Modifier(Gameplay=$System.IGPOFix, Front End=$System.FEPoFix)
       Destination Space=::Simulation
      )
     )
     Channel=::NewEnumerator2
    )
    Index 3 Param 1=&FloatFromCurve(FloatCurve=DataInterface() {propertyValues {Curve {keys [
         <at 0 -625.0001 RCIM_Cubic>
         {arriveTangent=712.5001, interpMode=RCIM_Cubic, leaveTangent=712.5001, time=1, value=0}
         {arriveTangent=1495.8616, interpMode=RCIM_Cubic, leaveTangent=1495.8616, tangentMode=RCTM_User, time=2, value=800}
        ]}}})
    Module.Param0WriteEnabled=true
    Module.Param0WriteEnabled2=true
    Module.Param0WriteEnabled3=true
    Module.Param1WriteEnabled=true
    Module.Param1WriteEnabled2=true
    Module.Param1WriteEnabled3=true
    Module.Param2WriteEnabled=true
    Module.Param2WriteEnabled2=true
    Module.Param2WriteEnabled3=true
    Module.Param3WriteEnabled=true
    Module.Param3WriteEnabled2=false
    Module.Param3WriteEnabled3=true
    Write Parameter Index 2=true
    Write Parameter Index 3=true
   ) {module=DynamicMaterialParameters001}
   &InterpolateOverTime(
    Rate Of Change:NiagaraFloat=&FN_FrontEndFloatModifier(Gameplay=10, Front End=10)
    Target Vector Value:Vector3f=&AddVectorToPosition(
     Position=$Particles.Position
     Vector=&TransformVector(
      Vector=&FN_FrontEndVector3Modifier(Gameplay=$System.IGPOFix, Front End=$System.FEPoFix)
      Destination Space=::Simulation
     )
    )
    Variable Type=::NewEnumerator2
   )
   /SetVariables(Particles.InterpolateDynamic:Vector4f=&MakeFloat4FromVectorAndFloat(
     "Vector (RGB)"=$StackContext.InterpolateOverTime.InterpolateOverTime_V3_.Moving Average
     "Float (Alpha)"=&FN_FrontEndFloatModifier(Gameplay=2, Front End=3)
    )) {isSetParameters=true}
   &SolveForcesAndVelocity(Clamp Velocity=false, Limit Acceleration=false)
   &ScaleColor(
    Scale RGB=&MakeVectorFromLinearColorRGB(Color=&FNTimeOfDayColor(Night Color=<rgba 0.3 0.475 1 1>, Morning Color=<rgba 1 1 1 1>, Day Color=<rgba 1 1 1 1>, Evening Color=<rgba 1 1 1 1>))
    ScaleA=true
    ScaleRGB=true
   )
  ]
  rendererClasses [&NiagaraMeshRendererProperties]
  renderers [{rendererClass=&NiagaraMeshRendererProperties, rendererIndex=0}]
 }
}
summary {
 systemName=NS_Glider_MistMylar_Idle
 emitters [
  {emitterName=Center, rendererClasses [&NiagaraMeshRendererProperties]}
  {emitterName=Trail_Dark, rendererClasses [&NiagaraMeshRendererProperties]}
  {emitterName=Smoke, rendererClasses [&NiagaraMeshRendererProperties]}
  {emitterName=Trail_Edge, rendererClasses [&NiagaraMeshRendererProperties]}
  {emitterName=CenterMask, rendererClasses [&NiagaraMeshRendererProperties]}
  {emitterName=Trail, rendererClasses [&NiagaraMeshRendererProperties]}
  {emitterName=Trail002, rendererClasses [&NiagaraMeshRendererProperties]}
 ]
 userVariables [
  User.bIsFrontEnd {type=NiagaraBool}
  User.bIsFrontEndPreview {type=NiagaraBool}
  User.VariantSwitch {type=NiagaraFloat, defaultValue=0, description=""}
 ]
}
